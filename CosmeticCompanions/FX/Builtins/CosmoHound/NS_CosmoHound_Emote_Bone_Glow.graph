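# Niagara effect for a glowing bone in a CosmoHound companion emote, built from four emitters.
# BoneMeshes emits short-lived meshes sampled from the skeletal mesh, slerping rotation from spawn to current pose, while Minimal drives a spinning, scaling mesh with animated material parameters.
# CubicFlameGlow layers flame and soft-glow sprites whose glow size follows a smoothed System.BuildUp switch.
# Spawn adds glitch sprites, ray meshes and lights, using spawn groups as visibility tags.
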NiagaraSystem /CosmeticCompanions/FX/Builtins/CosmoHound/NS_CosmoHound_Emote_Bone_Glow
dependencies {
 usedDataInterfaces [
  &NiagaraDataInterfaceCurve
  &NiagaraDataInterfaceVector2DCurve
  &NiagaraDataInterfaceVectorCurve
  &NiagaraDataInterfaceColorCurve
  &NiagaraDataInterfaceSkeletalMesh
 ]
 usedDynamicInputs [&MaskVector2dBySpawnGroup, &SetBoolByFloatComparison]
 usedModules [
  &self:SystemSpawnScript.SystemScriptSource.SystemScriptGraph.NiagaraNodeAssignment_0.SetVariables_#1
  &self:SystemSpawnScript.SystemScriptSource.SystemScriptGraph.NiagaraNodeAssignment_1.SetVariables_#2
  &SystemState
  &self:SystemSpawnScript.SystemScriptSource.SystemScriptGraph.NiagaraNodeAssignment_2.SetVariables_#3
  &self:SystemSpawnScript.SystemScriptSource.SystemScriptGraph.NiagaraNodeAssignment_3.SetVariables_#4
  &self:SystemSpawnScript.SystemScriptSource.SystemScriptGraph.NiagaraNodeAssignment_5.SetVariables_#5
 ]
}
emitters {
 BoneMeshes {
  emitterUpdateScript [
   &EmitterState(Life Cycle Mode=::NewEnumerator1, Loop Behavior=::NewEnumerator1, Loop Delay=3.6, UseLoopDelay=true)
   &SpawnRate(SpawnRate=30, Use Spawn Probability=false)
  ]
  particleSpawnScript [
   &InitializeParticle(Color Mode=::NewEnumerator0, Lifetime=0.3, Mesh Scale Mode=::NewEnumerator3, UsePositionOffset=false, Write Material Random=true)
   &SkeletalMeshLocation(
    Bone Sampling Mode=::NewEnumerator4
    MaskSpawnGroup=false
    Orientation Sampling=::NewEnumerator1
    Position Sampling=::NewEnumerator2
    Skeletal Mesh=$System.Skeletal Mesh
   )
  ]
  particleUpdateScript [
   &ParticleState()
   &ScaleMeshSize(Scale Factor=&Multiply_Vector(A=<1 1 1>, B=$Engine.Owner.Scale))
   &SkeletalMeshLocation(Bone Sampling Mode=::NewEnumerator4, MaskSpawnGroup=false, Orientation Sampling=::NewEnumerator1, Skeletal Mesh=$System.Skeletal Mesh) {module=SkeletalMeshLocation001}
   /SetVariables(Particles.MeshOrientation=&LerpQuaternion(
     Quaternion A:Quat4f=$Particles.Initial.SkeletalMeshLocation.SampledRotation
     Quaternion B:Quat4f=$Particles.SkeletalMeshLocation001.SampledRotation
     Lerp Factor:NiagaraFloat=&FloatFromCurve(FloatCurve=DataInterface() {propertyValues {Curve {
         defaultValue=1
         keys [
          {interpMode=RCIM_Cubic, tangentMode=RCTM_User, time=0, value=0}
          {
           arriveTangent=0.9358
           arriveTangentWeight=0.8153
           interpMode=RCIM_Cubic
           leaveTangent=0.9358
           leaveTangentWeight=0.8153
           tangentMode=RCTM_User
           tangentWeightMode=RCTWM_WeightedBoth
           time=1
           value=1
          }
         ]
        }}})
     Enable Spherical Linear Interpolation:NiagaraBool=true
    )) {isSetParameters=true}
  ]
  rendererClasses [&NiagaraMeshRendererProperties]
  renderers [{rendererClass=&NiagaraMeshRendererProperties, rendererIndex=0}]
 }
 CubicFlameGlow {
  emitterUpdateScript [
   &EmitterState(Life Cycle Mode=::NewEnumerator1, Loop Behavior=::NewEnumerator1, Loop Duration=10, UseLoopDelay=false)
   &SpawnBurst_Instantaneous(Spawn Group=2, Use Spawn Probability=false, UseLoopCountLimit=false) {module=SpawnBurst_Instantaneous001}
   &SpawnBurst_Instantaneous(Spawn Count=3, Spawn Group=1, Use Spawn Probability=false, UseLoopCountLimit=false)
   &SpawnRate(
    SpawnRate=&FloatFromCurve(
     FloatCurve=DataInterface() {propertyValues {Curve {keys [
         {interpMode=RCIM_Cubic, tangentMode=RCTM_User, time=0, value=0.25}
         {interpMode=RCIM_Cubic, tangentMode=RCTM_User, time=0.05, value=1}
         {interpMode=RCIM_Cubic, tangentMode=RCTM_User, time=0.75, value=1}
         {arriveTangent=-5.7829, tangentMode=RCTM_User, time=1, value=0.25}
        ]}}}
     CurveIndex=$Emitter.Age
     Scale Curve=40
    )
    Use Spawn Probability=true
   )
  ]
  particleSpawnScript [
   &PartitionParticles(Number Of Partitions=3, Partition Mode:ENiagara_PartitionMode_v2=::NewEnumerator5)
   &InitializeParticle(
    Color Mode=::NewEnumerator0
    Lifetime=&MaskFloat(
     Mask Source=0
     Mask Parameter=?
     Mask=0
     Masked Float=&RandomRangeFloat(Minimum=0.7, Maximum=1)
     Passthrough Float=5
    )
    Sprite Rotation Mode=::NewEnumerator2
    Sprite Size Mode=::NewEnumerator3
    Uniform Sprite Size=&FloatFromCurve(
     FloatCurve=DataInterface() {propertyValues {Curve {keys [{arriveTangent=0.0196, interpMode=RCIM_Cubic, leaveTangent=0.0196, tangentMode=RCTM_User, time=0, value=1}, <at 3 2>]}}}
     CurveIndex=$Emitter.Age
     Scale Curve=30
    )
    UsePositionOffset=false
    Write Material Random=true
   )
   /SetVariables(Particles.VisibilityTag=&SpawnGroup(Max Spawn Group=3)) {isSetParameters=true}
   &ShapeLocation(
    Cylinder Height=28
    Cylinder Radius=1
    MaskSpawnGroup=false
    Override Seed=false
    Rotation Coordinate Space=::Simulation
    Rotation Mode=::NewEnumerator2
    Shape Primitive=::NewEnumerator1
    Surface Only=false
    UseOwnerScale=true
    Yaw / Pitch / Roll=<0 0.25 0>
   )
   /SetVariables() {isSetParameters=true}
   &DynamicMaterialParameters(
    Index 0 Param 1=&RandomRangeFloat(Minimum=-0.05, Maximum=0.05)
    Index 0 Param 2=&RandomRangeFloat(Minimum=-0.05, Maximum=0.05)
    Index 0 Param 3=&RandomRangeFloat(Minimum=0, Maximum=1)
    Index 0 Param 4=&RandomRangeFloat(Minimum=0, Maximum=1)
    Module.Index0Float0Write=true
    Module.Index0Float1Write=true
    Module.Index0Float2Write=true
    Module.Index0Float3Write=true
   )
   /SetVariables(Particles.GlowSize=&Multiply_Vector2D_ByFloat(Float=&RandomRangeFloat(Minimum=1, Maximum=1.5), Vector2D=$Particles.SpriteSize)) {isSetParameters=true}
   /SetVariables(
    Particles.SoftGlowBGColor:LinearColor=<rgba 0.005 0.002 0.0005 0.1>
    Particles.SoftGlowSize:Vector2f=&Multiply_Vector2D_ByFloat(Float=2, Vector2D=$Particles.SpriteSize)
   ) {isSetParameters=true}
   &CameraOffset(Camera Offset Amount=-30)
  ]
  particleUpdateScript [
   &ParticleState()
   &ScaleSpriteSize(Uniform Curve Sprite Scale=DataInterface() {propertyValues {Curve {keys [
        {arriveTangent=1.4853, interpMode=RCIM_Cubic, leaveTangent=1.4853, tangentMode=RCTM_User, time=0, value=0}
        {tangentMode=RCTM_User, time=1, value=1}
       ]}}})
   &ScaleColor(
    Scale Alpha=&FloatFromCurve(FloatCurve=DataInterface() {propertyValues {Curve {keys [{interpMode=RCIM_Cubic, tangentMode=RCTM_User, time=0, value=0}, {interpMode=RCIM_Cubic, tangentMode=RCTM_User, time=0.1, value=1}, <at 1 0>]}}})
    Scale RGB=&Multiply_VectorByFloat(
     Vector=<1 1 1>
     Float=&FloatFromCurve(FloatCurve=DataInterface() {propertyValues {Curve {keys [
          {interpMode=RCIM_Cubic, tangentMode=RCTM_User, time=0, value=1}
          {interpMode=RCIM_Cubic, tangentMode=RCTM_User, time=0.084, value=3}
          <at 1 0>
         ]}}})
    )
    ScaleA=true
    ScaleRGB=true
   )
   &Drag(Use Linear Drag=true, Use Rotational Drag=true)
   &SolveForcesAndVelocity(Clamp Velocity=false, Limit Acceleration=false)
   /SetVariables(
    Particles.SoftGlowColor:LinearColor=&Multiply_LinearColorByFloat(
     Linear Color=<rgba 1.5 0.6142 0.075 1>
     Float=&FloatFromCurve(
      FloatCurve=DataInterface() {propertyValues {Curve {keys [
          {interpMode=RCIM_Cubic, tangentMode=RCTM_User, time=0, value=0}
          {arriveTangent=1.7763, interpMode=RCIM_Cubic, leaveTangent=1.7763, tangentMode=RCTM_User, time=0.0515, value=0.4165}
          {arriveTangentWeight=0.3202, interpMode=RCIM_Cubic, leaveTangentWeight=0.1067, tangentMode=RCTM_Break, time=0.8, value=1}
          <at 1 0>
         ]}}}
      Scale Curve=0.2
     )
    )
    Particles.SoftGlowSize:Vector2f=&Multiply_Vector2D_ByFloat(
     Float=&FloatFromCurve(
      FloatCurve=DataInterface() {propertyValues {Curve {
         defaultValue=1
         keys [
          {arriveTangent=5.5157, interpMode=RCIM_Cubic, leaveTangent=5.5157, tangentMode=RCTM_User, time=0, value=0.2}
          {arriveTangent=0.8, interpMode=RCIM_Cubic, leaveTangent=0.8, time=0.0793, value=0.4856}
          {
           arriveTangent=-0.0118
           arriveTangentWeight=0.2802
           interpMode=RCIM_Cubic
           leaveTangent=-0.0118
           leaveTangentWeight=0.2802
           tangentMode=RCTM_User
           tangentWeightMode=RCTWM_WeightedBoth
           time=1
           value=1
          }
         ]
        }}}
      Scale Curve=&Waveform("Waveform [1]"=::NewEnumerator0, "[1] Frequency"=5, "[1] Amplitude Min/Max"=<0.8 1>)
     )
     Vector2D=$Particles.Initial.SoftGlowSize
    )
   ) {isSetParameters=true}
   /SetVariables(
    Particles.GlowSize=&Multiply_Vector2D_ByFloat(
     Float=&Lerp_Float(
      A=&FloatFromCurve(FloatCurve=DataInterface() {propertyValues {Curve {
          defaultValue=1
          keys [
           {arriveTangent=3.387, interpMode=RCIM_Cubic, leaveTangent=3.387, tangentMode=RCTM_User, time=0, value=0}
           {
            arriveTangentWeight=0.4977
            interpMode=RCIM_Cubic
            leaveTangentWeight=0.4977
            tangentMode=RCTM_User
            tangentWeightMode=RCTWM_WeightedBoth
            time=1
            value=1
           }
          ]
         }}})
      B=&FloatFromCurve(FloatCurve=DataInterface() {propertyValues {Curve {
          defaultValue=1
          keys [
           {
            arriveTangent=-2.3577
            arriveTangentWeight=0.5421
            interpMode=RCIM_Cubic
            leaveTangent=-2.3577
            leaveTangentWeight=0.5421
            tangentMode=RCTM_User
            tangentWeightMode=RCTWM_WeightedBoth
            time=0
            value=1
           }
           {
            arriveTangent=-0.0005
            arriveTangentWeight=0.6223
            interpMode=RCIM_Cubic
            leaveTangent=-0.0005
            leaveTangentWeight=0.6223
            tangentMode=RCTM_User
            tangentWeightMode=RCTWM_WeightedBoth
            time=0.9324
            value=0
           }
          ]
         }}})
      Alpha=/Niagara/DynamicInputs/Lerp/SmoothLerpOverTime_Float(
       Convergence Rate:NiagaraFloat=50
       Smooth Value:NiagaraFloat=0.1
       Target Value:NiagaraFloat=&MakeCustomFloatFromBool(Bool=$System.BuildUp, True Float=1, False Float=0)
      )
     )
     Vector2D=$Particles.Initial.GlowSize
    )
    Particles.SoftGlowBGSize:Vector2f=&Multiply_Vector2D_ByFloat(Float=1.5, Vector2D=$Particles.SoftGlowSize)
   ) {isSetParameters=true}
   &Light_Attributes(
    Color=1
    Color Value To Scale:LinearColor=<rgba 5 2.3231 0 1>
    Diffuse Scale=1
    Falloff Exponent=1
    Light Diffuse Scale=100
    Light Exponent=5
    Light Radius:NiagaraFloat=&FloatFromCurve(
     FloatCurve=DataInterface() {propertyValues {Curve {keys [
         {arriveTangent=92.9213, interpMode=RCIM_Cubic, leaveTangent=92.9213, tangentMode=RCTM_User, time=0, value=0}
         {arriveTangentWeight=0.0193, interpMode=RCIM_Constant, leaveTangentWeight=0.244, time=0.0578, value=1}
         {arriveTangent=0.672, interpMode=RCIM_Cubic, leaveTangent=0.672, time=0.6, value=1}
         {arriveTangent=4.0728, interpMode=RCIM_Cubic, leaveTangent=4.0728, time=0.616, value=1.3751}
         {arriveTangent=-0.1955, interpMode=RCIM_Cubic, leaveTangent=-0.1955, time=0.6398, value=1.1622}
         <at 1 1.3>
        ]}}}
     Scale Curve=50
    )
    Radius=1
    Scale RGB:Vector3f=&Multiply_VectorByFloat(
     Vector=<1 1 1>
     Float=&Waveform(
      Global Amplitude Scale=&FloatFromCurve(FloatCurve=DataInterface() {propertyValues {Curve {keys [
           <at 0 1>
           {interpMode=RCIM_Cubic, tangentMode=RCTM_User, time=0.6, value=1}
           {arriveTangent=2.8433, interpMode=RCIM_Cubic, leaveTangent=2.8433, time=0.665, value=2}
           <at 0.7061 1.3018 RCIM_Cubic>
           <at 1 2>
          ]}}})
      "Waveform [1]"=::NewEnumerator0
      "[1] Frequency"=5
      "[1] Amplitude Min/Max"=<0.5 1>
     )
    )
    ScaleA=false
    ScaleRGB=true
   )
  ]
  rendererClasses [&NiagaraSpriteRendererProperties]
  renderers [
   {rendererClass=&NiagaraSpriteRendererProperties}
   {rendererClass=&NiagaraSpriteRendererProperties, rendererIndex=1}
  ]
 }
 Minimal {
  emitterUpdateScript [
   &EmitterState(Life Cycle Mode=::NewEnumerator1, Loop Behavior=::NewEnumerator1, Loop Delay=1, Loop Duration=4.3667, UseLoopDelay=false)
   &FN_Timer()
  ]
  particleSpawnScript [
   &InitializeParticle(Color=<rgba 1 0.1988 0.004 1>, Mesh Scale Mode=::NewEnumerator3, UsePositionOffset=false, Write Material Random=true)
   &InitialMeshOrientation(Mesh Orientation Mode=::NewEnumerator2, Use Rotation Vector=true)
  ]
  particleUpdateScript [
   &ParticleState()
   &ScaleMeshSize(Scale Factor=&VectorFromCurve(
     VectorCurve=DataInterface() {propertyValues {
       XCurve {keys [{interpMode=RCIM_Cubic, tangentMode=RCTM_User, time=0, value=1}, <at 1 0.8>]}
       YCurve {keys [
         {interpMode=RCIM_Cubic, tangentMode=RCTM_User, time=0, value=1}
         {interpMode=RCIM_Cubic, tangentMode=RCTM_User, time=0.8, value=1.3}
         <at 1 0.3>
        ]}
       ZCurve {keys [
         {interpMode=RCIM_Cubic, tangentMode=RCTM_User, time=0, value=1}
         {interpMode=RCIM_Cubic, tangentMode=RCTM_User, time=0.8, value=1.3}
         <at 1 0.3>
        ]}
      }}
     Scale Curve=<1 0.75 0.75>
    ))
   &UpdateMeshOrientation(
    Rotation Rate=-3
    Rotation Vector=/Niagara/DynamicInputs/ScaleByCurve/ScaleVectorByCurve(
     InitialValue:Vector3f=<0 -1 0.5>
     ScaleCurve:NiagaraDataInterfaceVectorCurve=DataInterface() {propertyValues {
       XCurve {keys [{leaveTangentWeight=0.3365, time=0, value=0}, <at 1 1>]}
       YCurve {keys [<at 0 1>]}
       ZCurve {keys [{leaveTangentWeight=0.3333, time=0, value=0}, <at 0.5 1>]}
      }}
     CurveIndex:NiagaraFloat=$Particles.NormalizedAge
    )
   )
   &DynamicMaterialParameters(
    Index 2 Param 1=10
    Index 2 Param 4=1
    Index 3 Param 1=&Add_Float(A=&RandomRangeFloat(Minimum=0, Maximum=1), B=$Emitter.FN_Timer.Value)
    Index 3 Param 2=0.13
    Module.Index2Float0Write=true
    Module.Index2Float1Write=true
    Module.Index2Float2Write=true
    Module.Index2Float3Write=true
    Module.Index3Float0Write=true
    Module.Index3Float1Write=true
    Module.Index3Float2Write=false
    Module.Index3Float3Write=false
    Write Parameter Index 0=::NewEnumerator0
    Write Parameter Index 2=::NewEnumerator1
    Write Parameter Index 3=::NewEnumerator1
   )
   &SkeletalMeshLocation(Bone Sampling Mode=::NewEnumerator4, MaskSpawnGroup=false, Orientation Sampling=::NewEnumerator1, Skeletal Mesh=$System.Skeletal Mesh)
   &ScaleColor(
    Scale Alpha=&FloatFromCurve(FloatCurve=DataInterface() {propertyValues {Curve {keys [{interpMode=RCIM_Cubic, tangentMode=RCTM_User, time=0, value=0}, <at 0.3 1>]}}})
    ScaleA=true
    ScaleRGB=true
   )
  ]
  rendererClasses [&NiagaraMeshRendererProperties]
  renderers [{rendererClass=&NiagaraMeshRendererProperties, rendererIndex=0}]
 }
 Spawn {
  emitterUpdateScript [
   &EmitterState(Life Cycle Mode=::NewEnumerator1, Loop Behavior=::NewEnumerator1, Loop Duration=0.7, UseLoopDelay=false)
   &SpawnBurst_Instantaneous(Spawn Count=3, Use Spawn Probability=false, UseLoopCountLimit=false)
   &SpawnRate(SpawnGroup=1, SpawnRate=15, Use Spawn Probability=false)
   &SpawnBurst_Instantaneous(Spawn Group=2, Use Spawn Probability=false, UseLoopCountLimit=false) {module=SpawnBurst_Instantaneous001}
  ]
  particleSpawnScript [
   /SetVariables(
    Particles.GlitchColor:LinearColor=<rgba 5 1.0838 0 1>
    Particles.GlitchSize:Vector2f=&RandomRangeVector2D(Minimum=<5 30>, Maximum=<8 50>)
    Particles.VisibilityTag=&SpawnGroup(Max Spawn Group=3)
   ) {isSetParameters=true}
   &InitializeParticle(
    Color Mode=::NewEnumerator0
    Lifetime=0.3
    Mesh Scale Mode=::NewEnumerator3
    Sprite Rotation Mode=::NewEnumerator2
    Sprite Size Mode=::NewEnumerator3
    Sprite UV Mode=::NewEnumerator1
    Uniform Sprite Size=&Lerp_Float(A=15, B=30, Alpha=&MakeFloatFromInt(Int=&ReturnExecIndex()))
    UsePositionOffset=false
    Write Material Random=true
   )
   &ShapeLocation(
    Cylinder Height=&MaskFloat(Mask Source=0, Mask Parameter=?, Mask=1, Masked Float=25, Passthrough Float=0)
    Cylinder Radius=1
    MaskSpawnGroup=false
    Override Seed=false
    Rotation Coordinate Space=::Simulation
    Rotation Mode=::NewEnumerator2
    Shape Primitive=::NewEnumerator1
    Surface Only=false
    UseOwnerScale=true
    Yaw / Pitch / Roll=<0 0.25 0>
   )
   &DynamicMaterialParameters(
    Index 0 Param 3=&RandomRangeFloat(Minimum=-1, Maximum=1)
    Index 0 Param 4=&RandomRangeFloat(Minimum=-1, Maximum=1)
    Module.Index0Float0Write=true
    Module.Index0Float1Write=true
    Module.Index0Float2Write=true
    Module.Index0Float3Write=true
   )
   /Game/Effects/Niagara/Modules/Sprite/FN_SpriteUV(End Frame=3, Write to Module Namepsace=true)
   &InheritVelocity(Speed Limit=true, UseSpeedThreshold=false)
   &ApplyInitialForces()
   /SetVariables(Particles.Rays DynP0:Vector4f=&MakeVector4(X=&RandomRangeFloat(Minimum=0, Maximum=1), Y=8, Z=2, W=1)) {isSetParameters=true}
  ]
  particleUpdateScript [
   &ParticleState()
   &ScaleSpriteSize(Uniform Curve Sprite Scale=DataInterface() {propertyValues {Curve {keys [
        {arriveTangent=23.1018, interpMode=RCIM_Cubic, leaveTangent=23.1018, tangentMode=RCTM_User, time=0, value=0.2}
        {interpMode=RCIM_Cubic, tangentMode=RCTM_User, time=0.3, value=3}
        {tangentMode=RCTM_User, time=1, value=0}
       ]}}})
   &ScaleColor(
    Scale Alpha=&FloatFromCurve(FloatCurve=DataInterface() {propertyValues {Curve {
        defaultValue=1
        keys [{interpMode=RCIM_Cubic, tangentMode=RCTM_User, time=0, value=0}, {interpMode=RCIM_Cubic, tangentMode=RCTM_User, time=0.3, value=0.7}]
       }}})
    Scale RGB=&Multiply_VectorByFloat(
     Vector=<1 1 1>
     Float=&FloatFromCurve(FloatCurve=DataInterface() {propertyValues {Curve {keys [<at 0 5 RCIM_Cubic>, {arriveTangent=2, interpMode=RCIM_Cubic, leaveTangent=2, time=0.243, value=2}, <at 1 7 RCIM_Cubic>]}}})
    )
    ScaleA=true
    ScaleRGB=true
   )
   /SetVariables(
    Particles.RaysColor:LinearColor=&Multiply_LinearColorByFloat(
     Linear Color=&MultiplyLinearColor(
      Linear Color=$Particles.Initial.GlitchColor
      Scale Factor=&Vector4FromCurve(Vector4Curve=DataInterface() {propertyValues {
         AlphaCurve {keys [<at 0 0.2>,<at 0.5582 0.2>,<at 1 0.5>]}
         BlueCurve {keys [{interpMode=RCIM_Cubic, tangentMode=RCTM_User, time=0.5595, value=1}, {arriveTangent=18.8498, tangentMode=RCTM_User, time=1, value=5}]}
         GreenCurve {keys [{interpMode=RCIM_Cubic, tangentMode=RCTM_User, time=0.5595, value=1}, {arriveTangent=18.8498, tangentMode=RCTM_User, time=1, value=5}]}
         RedCurve {keys [{interpMode=RCIM_Cubic, tangentMode=RCTM_User, time=0.5595, value=1}, {arriveTangent=18.8498, tangentMode=RCTM_User, time=1, value=5}]}
        }})
     )
     Float=1
    )
    Particles.RaysScale:Vector3f=&Multiply_VectorByFloat(
     Vector=$Particles.Initial.Scale
     Float=&FloatFromCurve(FloatCurve=DataInterface() {propertyValues {Curve {keys [
          {arriveTangent=2.1257, interpMode=RCIM_Cubic, leaveTangent=2.1257, tangentMode=RCTM_User, time=0, value=0.3}
          {interpMode=RCIM_Cubic, tangentMode=RCTM_User, time=0.4, value=1}
          <at 1 0>
         ]}}})
    )
   ) {isSetParameters=true}
   /SetVariables(Particles.Glitch DynP0:Vector4f=&MakeVector4(
     X=&RandomRangeFloat(Minimum=0, Maximum=1)
     Y=&Lerp_Float(A=0, B=0.5, Alpha=$Particles.NormalizedAge)
     Z=/Niagara/DynamicInputs/Integer/V2/Round_Float(Float:NiagaraFloat=&FloatFromCurve(FloatCurve=DataInterface() {propertyValues {Curve {defaultValue=1, keys [<at 0 0>,<at 1 1>]}}}))
     W=0
    )) {isSetParameters=true}
   &Light_Attributes(
    Color=1
    Color Value To Scale:LinearColor=&Multiply_LinearColorByFloat(Linear Color=<rgba 3 1.3939 0 1>, Float=1)
    Light Radius:NiagaraFloat=&FloatFromCurve(
     FloatCurve=DataInterface() {propertyValues {Curve {keys [<at 0 1>, {interpMode=RCIM_Cubic, tangentMode=RCTM_User, time=1, value=0.1}]}}}
     Scale Curve=150
    )
    Radius=1
    ScaleA=true
    ScaleRGB=false
   )
  ]
  rendererClasses [
   &NiagaraSpriteRendererProperties
   &NiagaraMeshRendererProperties
   &NiagaraLightRendererProperties
  ]
  renderers [
   {rendererClass=&NiagaraSpriteRendererProperties}
   {rendererClass=&NiagaraSpriteRendererProperties, rendererIndex=1}
   {rendererClass=&NiagaraMeshRendererProperties, rendererIndex=2}
   {rendererClass=&NiagaraLightRendererProperties, rendererIndex=3}
  ]
 }
}
summary {
 systemName=NS_CosmoHound_Emote_Bone_Glow
 emitters [
  {emitterName=CubicFlameGlow, rendererClasses [&NiagaraSpriteRendererProperties]}
  {emitterName=Minimal, rendererClasses [&NiagaraMeshRendererProperties]}
  {
   emitterName=Spawn
   rendererClasses [
    &NiagaraSpriteRendererProperties
    &NiagaraMeshRendererProperties
    &NiagaraLightRendererProperties
   ]
  }
  {emitterName=BoneMeshes, rendererClasses [&NiagaraMeshRendererProperties]}
 ]
 userVariables [User.AnimTime {type=NiagaraFloat, defaultValue=0, description=""}]
}
