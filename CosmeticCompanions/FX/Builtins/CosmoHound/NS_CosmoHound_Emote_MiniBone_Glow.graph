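# Niagara system for the glowing mini bone in a CosmoHound companion emote, built from three orange emitters.
# CubicFlameGlow spawns glow sprites along a thin cylinder for about eight seconds, with lights flickering on a hand-keyed curve.
# Glitch_Element fires two timed bursts of short-lived glitch sprites and expanding wave meshes.
# Trail draws eroding and lightning ribbons, pinning young particles to the emitter and stepping lightning parameters by loop age.
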NiagaraSystem /CosmeticCompanions/FX/Builtins/CosmoHound/NS_CosmoHound_Emote_MiniBone_Glow
dependencies {
 usedDataInterfaces [
  &NiagaraDataInterfaceCurve
  &NiagaraDataInterfaceVector2DCurve
  &NiagaraDataInterfaceColorCurve
  &NiagaraDataInterfaceSkeletalMesh
 ]
 usedDynamicInputs [
  &MaskVector2dBySpawnGroup
  &RandomRangeVector2D
  &FN_FrontEndFloatModifier
 ]
 usedModules [
  &self:SystemSpawnScript.SystemScriptSource.SystemScriptGraph.NiagaraNodeAssignment_0.SetVariables_#1
  &self:SystemSpawnScript.SystemScriptSource.SystemScriptGraph.NiagaraNodeAssignment_1.SetVariables_#2
  &self:SystemSpawnScript.SystemScriptSource.SystemScriptGraph.NiagaraNodeAssignment_3.SetVariables_#3
  &SystemState
  &SampleSkeletalMesh
  &self:SystemSpawnScript.SystemScriptSource.SystemScriptGraph.NiagaraNodeAssignment_2.SetVariables_#4
 ]
}
emitters {
 CubicFlameGlow {
  emitterUpdateScript [
   &EmitterState()
   &SpawnBurst_Instantaneous(Spawn Group=1, Use Spawn Probability=false, UseLoopCountLimit=false)
   &SpawnRate(
    SpawnRate=&FloatFromCurve(
     FloatCurve=DataInterface() {propertyValues {Curve {keys [
         {interpMode=RCIM_Cubic, tangentMode=RCTM_User, time=0, value=0.25}
         {interpMode=RCIM_Cubic, tangentMode=RCTM_User, time=0.05, value=1}
         {interpMode=RCIM_Cubic, tangentMode=RCTM_User, time=7.5, value=1}
         {arriveTangent=-1.0874, tangentMode=RCTM_User, time=8, value=0.25}
        ]}}}
     CurveIndex=$Emitter.Age
     Scale Curve=50
    )
    Use Spawn Probability=true
   )
  ]
  particleSpawnScript [
   &PartitionParticles(Number Of Partitions=3, Partition Mode:ENiagara_PartitionMode_v2=::NewEnumerator5)
   &InitializeParticle(
    Color Mode=::NewEnumerator0
    Lifetime=&MaskFloatBySpawnGroup(Masked Float=4, Passthrough Float=&RandomRangeFloat(Minimum=0.7, Maximum=1), Spawn Group Mask=1)
    Sprite Rotation Mode=::NewEnumerator2
    Sprite Size Mode=::NewEnumerator3
    Sprite UV Mode=::NewEnumerator1
    Uniform Sprite Size=15
    UsePositionOffset=false
    Write Material Random=true
   )
   /SetVariables(Particles.VisibilityTag=&SpawnGroup(Max Spawn Group=3)) {isSetParameters=true}
   &ShapeLocation(
    Cylinder Height=20
    Cylinder Radius=0
    MaskSpawnGroup=false
    Override Seed=false
    Rotation Coordinate Space=::Simulation
    Rotation Mode=::NewEnumerator2
    Shape Primitive=::NewEnumerator1
    Surface Only=false
    UseOwnerScale=true
    Yaw / Pitch / Roll=<0 0.25 0>
   )
   /SetVariables(Particles.Position2:NiagaraPosition=$Particles.Position) {isSetParameters=true}
   &DynamicMaterialParameters(
    Index 0 Param 1=&RandomRangeFloat(Minimum=-0.05, Maximum=0.05)
    Index 0 Param 2=&RandomRangeFloat(Minimum=-0.05, Maximum=0.05)
    Index 0 Param 3=&RandomRangeFloat(Minimum=0, Maximum=1)
    Index 0 Param 4=&RandomRangeFloat(Minimum=0, Maximum=1)
    Module.Index0Float0Write=true
    Module.Index0Float1Write=true
    Module.Index0Float2Write=true
    Module.Index0Float3Write=true
   )
   /SetVariables(Particles.GlowSize=&Multiply_Vector2D_ByFloat(Float=&RandomRangeFloat(Minimum=1, Maximum=1.5), Vector2D=$Particles.SpriteSize)) {isSetParameters=true}
   /SetVariables(Particles.SoftGlowSize:Vector2f=&Multiply_Vector2D_ByFloat(Float=2, Vector2D=$Particles.SpriteSize)) {isSetParameters=true}
   &CameraOffset(Camera Offset Amount=-10)
  ]
  particleUpdateScript [
   &ParticleState()
   &ScaleSpriteSize(Uniform Curve Sprite Scale=DataInterface() {propertyValues {Curve {keys [
        {arriveTangent=1.4853, interpMode=RCIM_Cubic, leaveTangent=1.4853, tangentMode=RCTM_User, time=0, value=0.1}
        {tangentMode=RCTM_User, time=1, value=1}
       ]}}})
   &ScaleColor(
    Scale Alpha=&FloatFromCurve(FloatCurve=DataInterface() {propertyValues {Curve {keys [{interpMode=RCIM_Cubic, tangentMode=RCTM_User, time=0, value=0}, {interpMode=RCIM_Cubic, tangentMode=RCTM_User, time=0.1, value=1}, <at 1 0>]}}})
    Scale RGB=&Multiply_VectorByFloat(
     Vector=<1 1 1>
     Float=&FloatFromCurve(FloatCurve=DataInterface() {propertyValues {Curve {keys [
          {interpMode=RCIM_Cubic, tangentMode=RCTM_User, time=0, value=1}
          {interpMode=RCIM_Cubic, tangentMode=RCTM_User, time=0.084, value=3}
          <at 1 0>
         ]}}})
    )
    ScaleA=true
    ScaleRGB=true
   )
   &Drag(Use Linear Drag=true, Use Rotational Drag=true)
   &SolveForcesAndVelocity(Clamp Velocity=false, Limit Acceleration=false)
   /SetVariables(Particles.SoftGlowColor:LinearColor=&Multiply_LinearColorByFloat(
     Linear Color=<rgba 1.2 0.4914 0.06 1>
     Float=&FloatFromCurve(
      FloatCurve=DataInterface() {propertyValues {Curve {keys [
          {interpMode=RCIM_Cubic, tangentMode=RCTM_User, time=0, value=0}
          {interpMode=RCIM_Cubic, tangentMode=RCTM_User, time=0.05, value=1}
          {arriveTangentWeight=0.3202, interpMode=RCIM_Cubic, leaveTangentWeight=0.1067, tangentMode=RCTM_Break, time=0.8, value=1}
          <at 1 0>
         ]}}}
      Scale Curve=0.3
     )
    )) {isSetParameters=true}
   &Light_Attributes(
    Color=1
    Color Value To Scale:LinearColor=<rgba 5 2.3231 0 1>
    Diffuse Scale=1
    Falloff Exponent=1
    Light Diffuse Scale=5
    Light Radius:NiagaraFloat=&FloatFromCurve(
     FloatCurve=DataInterface() {propertyValues {Curve {keys [
         {interpMode=RCIM_Cubic, tangentMode=RCTM_User, time=0, value=1.5}
         {arriveTangentWeight=0.0193, interpMode=RCIM_Cubic, leaveTangentWeight=0.244, tangentMode=RCTM_User, time=0.0578, value=1}
         {arriveTangentWeight=0.3557, interpMode=RCIM_Cubic, leaveTangentWeight=0.1023, tangentMode=RCTM_User, time=0.7896, value=1}
         <at 1 0>
        ]}}}
     Scale Curve=50
    )
    Radius=1
    Scale RGB:Vector3f=&Multiply_VectorByFloat(
     Vector=<1 1 1>
     Float=&FloatFromCurve(FloatCurve=DataInterface() {propertyValues {Curve {
         defaultValue=1
         keys [
          {interpMode=RCIM_Cubic, tangentMode=RCTM_User, time=0, value=0}
          {arriveTangent=16.8277, interpMode=RCIM_Cubic, leaveTangent=16.8277, time=0.0144, value=0.6782}
          {arriveTangent=1.3421, interpMode=RCIM_Cubic, leaveTangent=1.3421, time=0.0336, value=0.5647}
          {arriveTangent=2.0135, interpMode=RCIM_Cubic, leaveTangent=2.0135, time=0.0553, value=0.7331}
          {arriveTangent=3.9289, interpMode=RCIM_Cubic, leaveTangent=3.9289, time=0.0742, value=0.6465}
          {arriveTangent=-1.3028, interpMode=RCIM_Cubic, leaveTangent=-1.3028, time=0.0821, value=0.8382}
          {arriveTangent=-1.8678, interpMode=RCIM_Cubic, leaveTangent=-1.8678, time=0.1016, value=0.6108}
          {arriveTangent=-0.1647, interpMode=RCIM_Cubic, leaveTangent=-0.1647, time=0.1223, value=0.7632}
          {arriveTangent=-1.0148, interpMode=RCIM_Cubic, leaveTangent=-1.0148, time=0.1422, value=0.6041}
          {arriveTangent=-0.4523, interpMode=RCIM_Cubic, leaveTangent=-0.4523, time=0.1511, value=0.7339}
          {arriveTangent=-0.0808, interpMode=RCIM_Cubic, leaveTangent=-0.0808, time=0.1721, value=0.5906}
          {arriveTangent=-0.8566, interpMode=RCIM_Cubic, leaveTangent=-0.8566, time=0.1969, value=0.7302}
          {arriveTangent=-0.1938, interpMode=RCIM_Cubic, leaveTangent=-0.1938, time=0.2231, value=0.5469}
          {arriveTangent=0.8271, interpMode=RCIM_Cubic, leaveTangent=0.8271, time=0.248, value=0.7203}
          {arriveTangent=0.7582, interpMode=RCIM_Cubic, leaveTangent=0.7582, time=0.271, value=0.5866}
          {arriveTangent=-1.6586, interpMode=RCIM_Cubic, leaveTangent=-1.6586, time=0.2919, value=0.7535}
          {arriveTangent=-0.5714, interpMode=RCIM_Cubic, leaveTangent=-0.5714, time=0.3179, value=0.5088}
          {arriveTangent=2.0337, interpMode=RCIM_Cubic, leaveTangent=2.0337, time=0.3414, value=0.7253}
          {arriveTangent=-0.8194, interpMode=RCIM_Cubic, leaveTangent=-0.8194, time=0.3671, value=0.6089}
          {arriveTangent=-1.68, interpMode=RCIM_Cubic, leaveTangent=-1.68, time=0.3873, value=0.6876}
          {arriveTangent=4.2477, interpMode=RCIM_Cubic, leaveTangent=4.2477, time=0.4016, value=0.5509}
          {arriveTangent=1.5508, interpMode=RCIM_Cubic, leaveTangent=1.5508, time=0.4272, value=0.8571}
          {arriveTangent=-1.7562, interpMode=RCIM_Cubic, leaveTangent=-1.7562, time=0.4533, value=0.6311}
          {arriveTangent=0.3761, interpMode=RCIM_Cubic, leaveTangent=0.3761, time=0.4771, value=0.7693}
          {arriveTangent=1.5639, interpMode=RCIM_Cubic, leaveTangent=1.5639, time=0.5034, value=0.65}
          {arriveTangent=-0.0428, interpMode=RCIM_Cubic, leaveTangent=-0.0428, time=0.5256, value=0.8451}
          {arriveTangent=-1.376, interpMode=RCIM_Cubic, leaveTangent=-1.376, time=0.5603, value=0.6476}
          {arriveTangent=-1.2416, interpMode=RCIM_Cubic, leaveTangent=-1.2416, time=0.5839, value=0.7648}
          {arriveTangent=-0.3766, interpMode=RCIM_Cubic, leaveTangent=-0.3766, time=0.6266, value=0.5653}
          {arriveTangent=0.0658, interpMode=RCIM_Cubic, leaveTangent=0.0658, time=0.6469, value=0.7411}
          {arriveTangent=-0.137, interpMode=RCIM_Cubic, leaveTangent=-0.137, time=0.6791, value=0.5688}
          {arriveTangent=-0.4186, interpMode=RCIM_Cubic, leaveTangent=-0.4186, time=0.7013, value=0.7336}
          {arriveTangent=-0.7387, interpMode=RCIM_Cubic, leaveTangent=-0.7387, time=0.7323, value=0.5465}
          {arriveTangent=0.0752, interpMode=RCIM_Cubic, leaveTangent=0.0752, time=0.7636, value=0.6876}
          {arriveTangent=-0.2471, interpMode=RCIM_Cubic, leaveTangent=-0.2471, time=0.7916, value=0.5509}
          {arriveTangent=-0.4463, interpMode=RCIM_Cubic, leaveTangent=-0.4463, time=0.828, value=0.6717}
          {arriveTangent=-0.501, interpMode=RCIM_Cubic, leaveTangent=-0.501, time=0.8663, value=0.5176}
          {arriveTangent=-0.1642, interpMode=RCIM_Cubic, leaveTangent=-0.1642, time=0.8901, value=0.6406}
          {arriveTangent=0.1857, interpMode=RCIM_Cubic, leaveTangent=0.1857, time=0.9196, value=0.5088}
          {arriveTangent=-6.3301, interpMode=RCIM_Cubic, leaveTangent=-6.3301, time=0.9591, value=0.6534}
          <at 1 0>
         ]
        }}})
    )
    ScaleA=true
    ScaleRGB=true
   )
  ]
  rendererClasses [&NiagaraSpriteRendererProperties, &NiagaraLightRendererProperties]
  renderers [
   {rendererClass=&NiagaraSpriteRendererProperties}
   {rendererClass=&NiagaraSpriteRendererProperties, rendererIndex=1}
   {rendererClass=&NiagaraLightRendererProperties, rendererIndex=2}
  ]
 }
 Glitch_Element {
  emitterUpdateScript [
   &EmitterState(Life Cycle Mode=::NewEnumerator1, Loop Behavior=::NewEnumerator1, Loop Duration=10, UseLoopDelay=false)
   &SpawnBurst_Instantaneous(Spawn Time=0.07, Use Spawn Probability=false, UseLoopCountLimit=false)
   &SpawnBurst_Instantaneous(Spawn Group=1, Spawn Time=4.5, Use Spawn Probability=false, UseLoopCountLimit=false) {module=SpawnBurst_Instantaneous001}
  ]
  particleSpawnScript [
   /SetVariables(Particles.VisibilityTag=$Emitter.SpawnGroup) {isSetParameters=true}
   &InitializeParticle(
    Color=<rgba 5 1.0838 0 1>
    Lifetime=0.3
    Mesh Scale Mode=::NewEnumerator3
    Sprite Size=<8 30>
    Sprite Size Mode=::NewEnumerator4
    Sprite UV Mode=::NewEnumerator1
    UsePositionOffset=false
    Write Material Random=true
   )
   &DynamicMaterialParameters(
    Index 0 Param 1=&RandomRangeFloat(Minimum=0, Maximum=1)
    Index 0 Param 2=&RandomRangeFloat(Minimum=8, Maximum=12)
    Index 0 Param 3=&FloatFromCurve(FloatCurve=DataInterface() {propertyValues {Curve {keys [{interpMode=RCIM_Cubic, tangentMode=RCTM_User, time=0, value=1.6}, {arriveTangent=7.3367, tangentMode=RCTM_User, time=1, value=5}]}}})
    Module.Index0Float0Write=true
    Module.Index0Float1Write=true
    Module.Index0Float2Write=true
    Module.Index0Float3Write=true
   )
   /Game/Effects/Niagara/Modules/Sprite/FN_SpriteUV(End Frame=3, Write to Module Namepsace=true)
   &InheritVelocity(Speed Limit=true, UseSpeedThreshold=false)
   &ApplyInitialForces()
   &SpriteFacingAndAlignment(Sprite Alignment=$Particles.Velocity, WriteAlignment=true, WriteFacing=false)
  ]
  particleUpdateScript [
   &ParticleState()
   /SetVariables(Particles.Glitch DynP:Vector4f=&MakeVector4(
     X=&RandomRangeFloat(Minimum=0, Maximum=1)
     Y=&Lerp_Float(A=0.5, B=1, Alpha=$Particles.NormalizedAge)
     Z=/Niagara/DynamicInputs/Integer/V2/Round_Float(Float:NiagaraFloat=&FloatFromCurve(
       FloatCurve=DataInterface() {propertyValues {Curve {defaultValue=1, keys [<at 0 1>,<at 0.2035 0>,<at 0.444 0.9782>,<at 0.7 0>,<at 1 0.25>]}}}
       Scale Curve=4
      ))
     W=-1
    )) {isSetParameters=true}
   &Drag(Drag=5, Use Linear Drag=true, Use Rotational Drag=false)
   &SolveForcesAndVelocity(Clamp Velocity=false, Limit Acceleration=false)
   /SetVariables(
    Particles.WaveColor:LinearColor=&Multiply_LinearColorByFloat(
     Linear Color=$Particles.Initial.Color
     Float=&FloatFromCurve(FloatCurve=DataInterface() {propertyValues {Curve {keys [{interpMode=RCIM_Cubic, tangentMode=RCTM_User, time=0, value=1}, <at 1 0>]}}})
    )
    Particles.WaveScale:Vector3f=&Multiply_VectorByFloat(
     Vector=$Particles.Scale
     Float=&FloatFromCurve(
      FloatCurve=DataInterface() {propertyValues {Curve {
         defaultValue=1
         keys [
          {arriveTangent=1.4958, interpMode=RCIM_Cubic, leaveTangent=1.4958, tangentMode=RCTM_User, time=0, value=0.1}
          {interpMode=RCIM_Cubic, tangentMode=RCTM_User, time=1, value=1}
         ]
        }}}
      Scale Curve=0.5
     )
    )
   ) {isSetParameters=true}
  ]
  rendererClasses [&NiagaraSpriteRendererProperties, &NiagaraMeshRendererProperties]
  renderers [
   {rendererClass=&NiagaraSpriteRendererProperties}
   {rendererClass=&NiagaraMeshRendererProperties, rendererIndex=1}
  ]
 }
 Trail {
  emitterUpdateScript [
   &EmitterState(Inactive Response=::NewEnumerator1, Life Cycle Mode=::NewEnumerator1, Loop Delay=$System.Delay, Loop Duration=1, UseLoopDelay=false)
   &SpawnBurst_Instantaneous(Use Spawn Probability=false, UseLoopCountLimit=false)
   &SpawnPerUnit(Module.Use Max Movement Threshold=true, Module.Use Movement Tolerance=false, Spawn Spacing=30, Use Spawn Probability=false)
   &SpawnRate(SpawnRate=120, Use Spawn Probability=false)
   /SetVariables(Emitter.ParticleRandom:NiagaraFloat=&RandomRangeFloat(Minimum=0, Maximum=1, Recalculate Random Each Loop=false)) {isSetParameters=true}
  ]
  particleSpawnScript [
   &InitializeParticle(
    Color=<rgba 1.2 0.5103 0.0816 1>
    Lifetime=0.4
    Ribbon Width=20
    Ribbon Width Mode=::NewEnumerator1
    UsePositionOffset=false
    Write Material Random=true
   )
   &AddVelocity(Velocity=<400 0 0>) {enabled=false}
   /SetVariables(Particles.LightningRibbonWidth:NiagaraFloat=&Multiply_Float(A=$Particles.RibbonWidth, B=0.1)) {isSetParameters=true}
  ]
  particleUpdateScript [
   &ParticleState()
   &DynamicMaterialParameters(
    Index 0 Param 1=1
    Index 0 Param 2=0.75
    Index 0 Param 3=1
    Index 0 Param 4=1
    Index 1 Param 1=1
    Index 1 Param 2=1
    Index 1 Param 3=1
    Index 1 Param 4=1
    Index 3 Param 1=1
    Index 3 Param 2=1
    Index 3 Param 3=1
    Index 3 Param 4=2.2631
    Module.Index0Param0Write=true
    Module.Index0Param1Write=true
    Module.Index0Param2Write=true
    Module.Index0Param3Write=true
    Module.Index1Param0Write=true
    Module.Index1Param1Write=true
    Module.Index1Param2Write=true
    Module.Index1Param3Write=true
    Module.Index2Param0Write=true
    Module.Index2Param1Write=true
    Module.Index2Param2Write=true
    Module.Index2Param3Write=true
    Module.Index3Param0Write=true
    Module.Index3Param1Write=true
    Module.Index3Param2Write=true
    Module.Index3Param3Write=true
    Write Parameter Index 1=true
    Write Parameter Index 2=true
    Write Parameter Index 3=true
   )
   &SolveForcesAndVelocity(Clamp Velocity=false, Limit Acceleration=false)
   &FN_RibbonMaterial(
    Color=&ColorFromCurve(ColorCurve=DataInterface() {propertyValues {
       AlphaCurve {keys [<at 0 1 RCIM_Cubic>]}
       BlueCurve {keys [<at 0 0.1347 RCIM_Cubic>]}
       GreenCurve {keys [<at 0 2.0772 RCIM_Cubic>]}
       RedCurve {keys [<at 0 3 RCIM_Cubic>]}
      }})
    Color Pan U Speed=1
    Color Scale U=0.3
    Erosion Amount:NiagaraFloat=&FloatFromCurve(FloatCurve=DataInterface() {propertyValues {Curve {
        defaultValue=1
        keys [
         {arriveTangent=2.9225, interpMode=RCIM_Cubic, leaveTangent=2.9225, tangentMode=RCTM_User, time=0, value=0}
         {interpMode=RCIM_Cubic, tangentMode=RCTM_User, time=1, value=1}
        ]
       }}})
    Erosion Pan Speed U=7
    Erosion Power=1.5
    Erossion Scale U=3
    Macro Distortion Amount=-0.1
    Macro Pan Speed=0.3
    Micro Distortion Amount=0.1
    Micro Pan Speed=0.4
    Micro Scale U=3
    Opacity Erosion=0.1
    Opacity Pan U Speed=0.5
    Opacity Power=3
    Opacity Scale U=0.3
    Scale Alpha:NiagaraFloat=&FloatFromCurve(FloatCurve=DataInterface() {propertyValues {Curve {keys [{interpMode=RCIM_Cubic, tangentMode=RCTM_User, time=0.3, value=1}, {interpMode=RCIM_Cubic, tangentMode=RCTM_User, time=1, value=0}]}}})
   )
   /SetVariables(Particles.Position=&Lerp_Vector(
     A=Linked() {linkedVariable {name=Engine.Emitter.SimulationPosition, type=NiagaraPosition}}
     B=Linked() {linkedVariable {name=Particles.Position, type=NiagaraPosition}}
     Alpha=&FloatFromCurve(FloatCurve=DataInterface() {propertyValues {Curve {
         defaultValue=1
         keys [{interpMode=RCIM_Cubic, tangentMode=RCTM_User, time=0.1, value=0}, {interpMode=RCIM_Cubic, tangentMode=RCTM_User, time=0.2, value=1}]
        }}})
    )) {isSetParameters=true}
   /SetVariables(Particles.LightningColor:LinearColor=&Multiply_LinearColorByFloat(Linear Color=<rgba 2 0.1311 0 1>, Float=5)) {isSetParameters=true}
   /SetVariables(
    Particles.Lightning Dyn00:Vector4f=&MakeVector4(
     X=&FloatFromCurve(FloatCurve=DataInterface() {propertyValues {Curve {keys [
          {arriveTangent=3.0269, interpMode=RCIM_Cubic, leaveTangent=3.0269, tangentMode=RCTM_User, time=0, value=0}
          {tangentMode=RCTM_User, time=1, value=1}
         ]}}})
     Y=0.7
     Z=&FloatFromCurve(
      FloatCurve=DataInterface() {propertyValues {Curve {keys [
          <at 0 0 RCIM_Cubic>
          {
           arriveTangent=0.7872
           arriveTangentWeight=0.0635
           interpMode=RCIM_Cubic
           tangentMode=RCTM_Break
           tangentWeightMode=RCTWM_WeightedBoth
           time=0.25
           value=5
          }
          <at 0.25 0 RCIM_Cubic>
          {
           arriveTangent=0.5982
           arriveTangentWeight=0.0581
           interpMode=RCIM_Cubic
           leaveTangentWeight=0.0003
           tangentMode=RCTM_Break
           tangentWeightMode=RCTWM_WeightedBoth
           time=0.5
           value=5
          }
          <at 0.501 0 RCIM_Cubic>
          {
           arriveTangent=0.6048
           arriveTangentWeight=0.0579
           interpMode=RCIM_Cubic
           leaveTangentWeight=0.0003
           tangentMode=RCTM_Break
           tangentWeightMode=RCTWM_WeightedBoth
           time=0.75
           value=5
          }
          <at 0.751 0 RCIM_Cubic>
          {
           arriveTangent=0.6048
           arriveTangentWeight=0.0579
           interpMode=RCIM_Cubic
           leaveTangent=1.3824
           leaveTangentWeight=0.0762
           tangentMode=RCTM_Break
           tangentWeightMode=RCTWM_WeightedBoth
           time=1
           value=5
          }
         ]}}}
      CurveIndex=$Emitter.NormalizedLoopAge
      Scale Curve=0.2
     )
     W=&FloatFromCurve(FloatCurve=DataInterface() {propertyValues {Curve {keys [
          {arriveTangent=2.5109, interpMode=RCIM_Cubic, leaveTangent=2.5109, tangentMode=RCTM_User, time=0, value=0}
          {interpMode=RCIM_Cubic, tangentMode=RCTM_User, time=1, value=1}
         ]}}})
    )
    Particles.Lightning Dyn01:Vector4f=&MakeVector4(
     X=&FloatFromCurve(FloatCurve=DataInterface() {propertyValues {Curve {keys [<at 0.25 1>,<at 1 0>]}}})
     Y=&Add_Float(
      A=$Emitter.ParticleRandom
      B=&FloatFromCurve(
       FloatCurve=DataInterface() {propertyValues {Curve {
          defaultValue=0
          keys [
           {
            arriveTangent=0.0962
            arriveTangentWeight=0.2726
            interpMode=RCIM_Cubic
            leaveTangentWeight=0.1272
            tangentMode=RCTM_Break
            tangentWeightMode=RCTWM_WeightedBoth
            time=0
            value=1
           }
           {
            arriveTangent=0.0962
            arriveTangentWeight=0.2726
            interpMode=RCIM_Cubic
            leaveTangentWeight=0.1272
            tangentMode=RCTM_Break
            tangentWeightMode=RCTWM_WeightedBoth
            time=0.25
            value=0
           }
           {
            arriveTangent=0.0962
            arriveTangentWeight=0.2726
            interpMode=RCIM_Cubic
            leaveTangentWeight=0.1272
            tangentMode=RCTM_Break
            tangentWeightMode=RCTWM_WeightedBoth
            time=0.5
            value=-1
           }
           {
            arriveTangent=0.0962
            arriveTangentWeight=0.2726
            interpMode=RCIM_Cubic
            leaveTangentWeight=0.1272
            tangentMode=RCTM_Break
            tangentWeightMode=RCTWM_WeightedBoth
            time=0.75
            value=-2
           }
           <at 1 -3>
          ]
         }}}
       CurveIndex=$Emitter.NormalizedLoopAge
       Scale Curve=0.5
      )
     )
     Z=1
     W=1
    )
    Particles.Lightning Dyn02:Vector4f=&MakeVector4(
     X=&FloatFromCurve(FloatCurve=DataInterface() {propertyValues {Curve {keys [{arriveTangent=-2.6239, interpMode=RCIM_Cubic, leaveTangent=-2.6239, tangentMode=RCTM_User, time=0, value=1}, <at 1 0.3>]}}})
     Y=0
     Z=0
     W=0
    )
   ) {isSetParameters=true}
  ]
  rendererClasses [&NiagaraRibbonRendererProperties]
  renderers [
   {rendererClass=&NiagaraRibbonRendererProperties}
   {rendererClass=&NiagaraRibbonRendererProperties, rendererIndex=1}
  ]
 }
}
summary {
 systemName=NS_CosmoHound_Emote_MiniBone_Glow
 emitters [
  {
   emitterName=CubicFlameGlow
   rendererClasses [&NiagaraSpriteRendererProperties, &NiagaraLightRendererProperties]
  }
  {emitterName=Trail, rendererClasses [&NiagaraRibbonRendererProperties]}
  {emitterName=Glitch_Element, rendererClasses [&NiagaraSpriteRendererProperties, &NiagaraMeshRendererProperties]}
 ]
 userVariables [
  User.bIsFrontEnd {type=NiagaraBool}
  User.bIsFrontEndPreview {type=NiagaraBool}
  User.TwoBones {type=NiagaraBool}
 ]
}
