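# Niagara system for the CosmoHound cosmetic companion's emote sequence, timing glows, glitches and lightning over several seconds.
# Dark_Glow fires staggered bursts, using spawn-group switches to shape sprite, mesh and light layers, with User.Variant choosing orange, blue or red palettes.
# Glitch adds an opening sprite burst, while Lightning builds randomized, mirrored mesh bolts through age-driven rates and bursts.
# Front-end modifiers adjust scale, brightness and orientation outside gameplay.
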NiagaraSystem /CosmeticCompanions/FX/Builtins/CosmoHound/NS_CosmoHound_Emote_Sequence
dependencies {
 usedDataInterfaces [
  &NiagaraDataInterfaceArrayFloat
  &NiagaraDataInterfaceCurve
  &NiagaraDataInterfaceVectorCurve
  &NiagaraDataInterfaceArrayColor
 ]
 usedDynamicInputs []
 usedModules [
  &self:SystemSpawnScript.SystemScriptSource.SystemScriptGraph.NiagaraNodeAssignment_0.SetVariables_#1
  &SystemState
 ]
}
emitters {
 Dark_Glow {
  emitterUpdateScript [
   &EmitterState(Inactive Response=::NewEnumerator1, Life Cycle Mode=::NewEnumerator1, Loop Behavior=::NewEnumerator1, Loop Duration=9, UseLoopDelay=false)
   &SpawnBurst_Instantaneous(Spawn Count=1, Spawn Time=3.5, Use Spawn Probability=false)
   &SpawnBurst_Instantaneous(Spawn Group=1, Spawn Time=3.65, Use Spawn Probability=false, UseLoopCountLimit=false) {module=SpawnBurst_Instantaneous001}
   &SpawnBurst_Instantaneous(Spawn Group=2, Spawn Time=4.3, Use Spawn Probability=false, UseLoopCountLimit=false) {module=SpawnBurst_Instantaneous002}
   &SpawnBurst_Instantaneous(Spawn Group=3, Spawn Time=1.3, Use Spawn Probability=false, UseLoopCountLimit=false) {module=SpawnBurst_Instantaneous003}
   &SpawnBurst_Instantaneous(Spawn Group=4, Spawn Time=3.5, Use Spawn Probability=false, UseLoopCountLimit=false) {module=SpawnBurst_Instantaneous004}
  ]
  particleSpawnScript [
   /SetVariables(Particles.VisibilityTag=&SpawnGroup()) {isSetParameters=true}
   /SetVariables(Particles.Random=&RandomRangeFloat(Minimum=0, Maximum=1)) {isSetParameters=true}
   &InitializeParticle(
    Color=&MergeHelper_ColorSwitch(
     0=<rgba 1 1 1 1.5>
     1=&SelectLinearColorFromArray(
      Array Sampling Mode=::NewEnumerator1
      Direct Array Index=&MakeIntFromFloat(Float=$User.Variant, Conversion Mode=::NewEnumerator1)
      Color Selection Array=DataInterface() {propertyValues {ColorData [<rgba 1.5 0.213 0.015 1>, <rgba 0 0.4916 1 1>, <rgba 1 0.0238 0.024 1>]}}
      Random Seed=?
     )
     2=&SelectLinearColorFromArray(
      Array Sampling Mode=::NewEnumerator1
      Direct Array Index=&MakeIntFromFloat(Float=$User.Variant, Conversion Mode=::NewEnumerator1)
      Color Selection Array=DataInterface() {propertyValues {ColorData [<rgba 1 0.3784 0 1>, <rgba 0 0.4916 1 1>, <rgba 1 0.0238 0.024 1>]}}
      Random Seed=?
     )
     3=&SelectLinearColorFromArray(
      Array Sampling Mode=::NewEnumerator1
      Direct Array Index=&MakeIntFromFloat(Float=$User.Variant, Conversion Mode=::NewEnumerator1)
      Color Selection Array=DataInterface() {propertyValues {ColorData [<rgba 1 0.3784 0 1>, <rgba 0 0.4916 1 1>, <rgba 1 0.0167 0 1>]}}
      Random Seed=?
     )
     4=&MergeHelper_ColorSwitch(
      0=<rgba 3 0.965 0 1>
      1=<rgba 0 1.0468 2 1>
      2=<rgba 3 0.0468 0 1>
      3=?
      Index=&MakeIntFromFloat(Float=$User.Variant, Conversion Mode=::NewEnumerator1)
      Num=::NewEnumerator1
     )
     5=&SelectLinearColorFromArray(
      Array Sampling Mode=::NewEnumerator1
      Direct Array Index=&MakeIntFromFloat(Float=$User.Variant, Conversion Mode=::NewEnumerator1)
      Color Selection Array=DataInterface() {propertyValues {ColorData [<rgba 1 0.298 0 1>, <rgba 0 0.3144 1 1>, <rgba 1 0.0001 0.0167 1>]}}
      Random Seed=?
     )
     6=&MergeHelper_ColorSwitch(
      0=<rgba 5 3.6617 0 1>
      1=<rgba 0 3.1555 5 1>
      2=<rgba 5 0.1522 0.1148 1>
      3=?
      Index=&MakeIntFromFloat(Float=$User.Variant, Conversion Mode=::NewEnumerator1)
      Num=::NewEnumerator1
     )
     Index=&SpawnGroup()
     Num=::NewEnumerator5
    )
    Lifetime=&MergeHelper_FloatSwitch(0=1, 1=1, 2=0.3, 3=3.3, 4=0.95, 5=0.5, 6=0.25, Index=&SpawnGroup(), Num=::NewEnumerator5)
    Mesh Scale=&MergeHelper_VectorSwitch(
     0=<0 0 0>
     1=<0 0 0>
     2=<1 1 1>
     3=<0 0 0>
     4=&Multiply_VectorByFloat(Vector=<-2 -4 -3.5>, Float=0.2)
     5=&Multiply_VectorByFloat(Vector=<-2 -4 -3.5>, Float=0.3)
     6=&Multiply_VectorByFloat(
      Vector=&Lerp_Vector(
       A=&RandomRangeVector(Minimum=<0.5 0.4 0.4>, Maximum=<0.75 0.6 0.6>)
       B=&RandomRangeVector(Minimum=<1 1 1>, Maximum=<1.5 2 2>)
       Alpha=&FloatFromCurve(
        FloatCurve=DataInterface() {propertyValues {Curve {keys [{interpMode=RCIM_Cubic, tangentMode=RCTM_User, time=0.5, value=0}, <at 1 1>]}}}
        CurveIndex=&RandomRangeFloat(Minimum=0, Maximum=1)
       )
      )
      Float=0.6
     )
     Index=&SpawnGroup()
     Num=::NewEnumerator5
    )
    Mesh Scale Mode=::NewEnumerator4
    Sprite Rotation Angle=&MergeHelper_FloatSwitch(0=0.12, 1=0.12, 2=0, 3=0, 4=0, Index=&SpawnGroup(), Num=::NewEnumerator3)
    Sprite Rotation Mode=::NewEnumerator1
    Sprite Size Mode=::NewEnumerator3
    Uniform Sprite Size=&MergeHelper_FloatSwitch(0=100, 1=44.2378, 2=60, 3=80, 4=0, Index=&SpawnGroup(), Num=::NewEnumerator3)
    UsePositionOffset=false
    Write Material Random=true
   )
   &DynamicMaterialParameters(
    Index 0 Param 1=&MergeHelper_FloatSwitch(0=2, 1=2, 2=3, 3=1, 4=0, 5=0, Index=&SpawnGroup(), Num=::NewEnumerator4)
    Index 0 Param 2=&MergeHelper_FloatSwitch(0=1, 1=1, 2=1, 3=1, 4=1, 5=0, Index=&SpawnGroup(), Num=::NewEnumerator4)
    Index 0 Param 3=&MergeHelper_FloatSwitch(0=1, 1=1, 2=1, 3=1, 4=1, 5=0, Index=&SpawnGroup(), Num=::NewEnumerator4)
    Index 0 Param 4=&MergeHelper_FloatSwitch(0=1, 1=1, 2=1, 3=1, 4=1, 5=0, Index=&SpawnGroup(), Num=::NewEnumerator4)
    Module.Index0Float0Write=true
    Module.Index0Float1Write=true
    Module.Index0Float2Write=true
    Module.Index0Float3Write=true
   )
   &CameraOffset(Camera Offset Amount=&MergeHelper_FloatSwitch(0=10, 1=50, 2=-3, 3=0, 4=0, Index=&SpawnGroup(), Num=::NewEnumerator3))
   &SpriteFacingAndAlignment(Sprite Facing=<0 1 0>, WriteAlignment=false, WriteFacing=true)
   &AddVelocity(
    Module.Use Falloff Distance=false
    Override Seed=false
    Velocity Mode=::NewEnumerator1
    Velocity Speed=&MergeHelper_FloatSwitch(0=0, 1=0, 2=0, 3=0, 4=0, 5=0, 6=-20, Index=&SpawnGroup(), Num=::NewEnumerator5)
   )
   /SetVariables(
    Particles.Rays DynP 0:Vector4f=&MakeVector4(X=1, Y=1, Z=1, W=3)
    Particles.Rays DynP 1:Vector4f=&MakeVector4(X=1, Y=0, Z=0, W=0)
    Particles.Rays DynP 2:Vector4f=&MakeVector4(X=3, Y=0, Z=1, W=1)
    Particles.Rays DynP 3:Vector4f=&MakeVector4(X=30, Y=0, Z=0, W=0)
   ) {isSetParameters=true}
   &InitialMeshOrientation(
    Mesh Orientation Mode=::NewEnumerator2
    Rotation=&MakeVector(X=&RandomRangeFloat(Minimum=0, Maximum=1), Y=0, Z=0)
    Use Rotation Vector=true
   )
  ]
  particleUpdateScript [
   &ParticleState()
   /SetVariables(Particles.DynPar_ChromaWave:Vector4f=&MakeVector4(
     X=&RandomRangeFloat(Minimum=0, Maximum=1)
     Y=&MergeHelper_FloatSwitch(
      0=0
      1=0
      2=0
      3=0
      4=&FloatFromCurve(
       FloatCurve=DataInterface() {propertyValues {Curve {
          defaultValue=1
          keys [
           {arriveTangent=0.2806, interpMode=RCIM_Cubic, leaveTangent=0.2806, tangentMode=RCTM_User, time=0, value=0}
           {arriveTangent=2.2839, interpMode=RCIM_Cubic, leaveTangent=2.2839, tangentMode=RCTM_User, time=1, value=1}
          ]
         }}}
       Scale Curve=0.5
      )
      5=&FloatFromCurve(
       FloatCurve=DataInterface() {propertyValues {Curve {
          defaultValue=1
          keys [
           {arriveTangent=-3.4505, interpMode=RCIM_Cubic, leaveTangent=-3.4505, leaveTangentWeight=0.8311, tangentMode=RCTM_User, time=0, value=1}
           {arriveTangent=-0.2806, arriveTangentWeight=0.3462, interpMode=RCIM_Cubic, leaveTangent=-0.2806, tangentMode=RCTM_User, time=1, value=0}
          ]
         }}}
       Scale Curve=0.7
      )
      Index=&SpawnGroup()
      Num=::NewEnumerator4
     )
     Z=0
     W=0
    )) {isSetParameters=true}
   /SetVariables(Particles.Lightning glow color:LinearColor=&MakeLinearColorFromVectorAndFloat(
     "Vector (RGB)"=&MakeVectorFromLinearColorRGB(Color=&MultiplyLinearColor(
       Linear Color=&SelectLinearColorFromArray(
        Array Sampling Mode=::NewEnumerator1
        Direct Array Index=&MakeIntFromFloat(Float=$User.Variant, Conversion Mode=::NewEnumerator1)
        Color Selection Array=DataInterface() {propertyValues {ColorData [<rgba 4 1.6071 0.1968 1>, <rgba 0.3366 2.0701 4 1>, <rgba 4 0.1073 0.1894 1>]}}
        Random Seed=?
       )
       Scale Factor=&MakeVector4_FromFloat(Float=&RandomRangeFloat(Minimum=0.5, Maximum=1, Evaluation Type=::NewEnumerator1))
      ))
     "Float (Alpha)"=&Multiply_Float(
      A=&FloatFromCurve(FloatCurve=DataInterface() {propertyValues {Curve {keys [<at 0 1 RCIM_Cubic>, {interpMode=RCIM_Cubic, leaveTangent=-3.6307, time=0.9, value=0}]}}})
      B=1
     )
    )) {isSetParameters=true}
   /SetVariables(Particles.Glow Dy02:Vector4f=&MakeVector4(
     X=&FloatFromCurve(FloatCurve=DataInterface() {propertyValues {Curve {keys [{tangentMode=RCTM_User, time=0, value=0.2}, {tangentMode=RCTM_User, time=0.5, value=1}]}}})
     Y=0.7
     Z=0
     W=5
    )) {isSetParameters=true}
   /SetVariables(Particles.Glow Dy03:Vector4f=&MakeVector4(
     X=&FloatFromCurve(
      FloatCurve=DataInterface() {propertyValues {Curve {keys [<at 0 1>,<at 1 0.5>]}}}
      Scale Curve=&MergeHelper_FloatSwitch(0=0.7, 1=0.5, 2=1.8, 3=?, Index=&MakeIntFromFloat(Float=$User.Variant), Num=::NewEnumerator1)
     )
     Y=$Particles.Random
     Z=1.3
     W=1
    )) {isSetParameters=true}
   /SetVariables(
    Particles.Lightning Dyn Param:Vector4f=&MakeVector4(
     X=&FloatFromCurve(FloatCurve=DataInterface() {propertyValues {Curve {keys [{tangentMode=RCTM_User, time=0, value=1}, {tangentMode=RCTM_User, time=1, value=1}]}}})
     Y=0.7
     Z=&FloatFromCurve(FloatCurve=DataInterface() {propertyValues {Curve {keys [
          {arriveTangent=4.0072, interpMode=RCIM_Cubic, leaveTangent=4.0072, tangentMode=RCTM_User, time=0, value=0}
          {arriveTangent=2.7083, interpMode=RCIM_Cubic, leaveTangent=2.7083, time=0.2543, value=0.6225}
          {arriveTangent=1.8473, interpMode=RCIM_Cubic, leaveTangent=1.8473, time=0.6, value=1.625}
          {arriveTangent=1.6993, interpMode=RCIM_Cubic, leaveTangent=1.6993, tangentMode=RCTM_User, time=1, value=2}
         ]}}})
     W=5
    )
    Particles.Lightning Dyn Param 01:Vector4f=&MakeVector4(
     X=&MergeHelper_FloatSwitch(
      0=0.7
      1=0.7
      2=0.5
      3=?
      Index=&MakeIntFromFloat(Float=$User.Variant, Conversion Mode=::NewEnumerator1)
      Num=::NewEnumerator1
     )
     Y=$Particles.Random
     Z=1.3
     W=1
    )
    Particles.Lightning Dyn Param 02:Vector4f=&MakeVector4(
     X=&FloatFromCurve(FloatCurve=DataInterface() {propertyValues {Curve {keys [<at 0 1>,<at 1 0.3>]}}})
     Y=0
     Z=0
     W=0
    )
   ) {isSetParameters=true}
   &SolveForcesAndVelocity(Clamp Velocity=false, Limit Acceleration=false)
   &SpriteRotationRate(Rotation Rate=&MergeHelper_FloatSwitch(
     0=0
     1=0
     2=0
     3=&FloatFromCurve(
      FloatCurve=DataInterface() {propertyValues {Curve {
         defaultValue=1
         keys [
          {interpMode=RCIM_Cubic, tangentMode=RCTM_User, time=0, value=0}
          {arriveTangent=-4.4276, interpMode=RCIM_Cubic, leaveTangent=-4.4276, time=0.0958, value=1.1104}
          {arriveTangentWeight=0.1138, interpMode=RCIM_Cubic, leaveTangentWeight=0.2881, tangentMode=RCTM_User, time=0.2911, value=-1.2889}
          {arriveTangentWeight=0.2866, interpMode=RCIM_Cubic, leaveTangentWeight=0.5185, tangentMode=RCTM_User, time=0.6308, value=0}
          <at 0.7688 0.18 RCIM_Cubic>
          {interpMode=RCIM_Cubic, tangentMode=RCTM_User, time=1, value=0}
         ]
        }}}
      Scale Curve=400
     )
     4=0
     Index=&SpawnGroup()
     Num=::NewEnumerator3
    ))
   &ScaleColor(
    Scale Alpha=&MergeHelper_FloatSwitch(
     0=&Multiply_Float(
      A=&FN_LOD_Distance_Remap(Start Distance=350, End Distance=380, Minimum Output=1, Maximum Output=1, Clamp Output=true)
      B=&FloatFromCurve(
       FloatCurve=DataInterface() {propertyValues {Curve {keys [
           <at 0.1602 0 RCIM_Cubic>
           {arriveTangentWeight=0.2318, interpMode=RCIM_Cubic, leaveTangentWeight=0.2396, tangentMode=RCTM_User, time=0.5, value=1}
           <at 1 1 RCIM_Cubic>
          ]}}}
       Scale Curve=1.1
      )
     )
     1=&Multiply_Float(
      A=&FN_LOD_Distance_Remap(Start Distance=350, End Distance=380, Minimum Output=1, Maximum Output=1, Clamp Output=true)
      B=&FloatFromCurve(
       FloatCurve=DataInterface() {propertyValues {Curve {keys [
           {arriveTangent=-0.0934, interpMode=RCIM_Cubic, time=0, value=0}
           {arriveTangentWeight=0.2318, interpMode=RCIM_Cubic, leaveTangentWeight=0.2396, tangentMode=RCTM_User, time=0.05, value=1}
           {arriveTangentWeight=1.1026, interpMode=RCIM_Cubic, leaveTangentWeight=0.2345, tangentMode=RCTM_User, time=0.86, value=0.9255}
           {arriveTangentWeight=1.1731, interpMode=RCIM_Cubic, leaveTangentWeight=0.1669, tangentMode=RCTM_User, time=0.9, value=5}
           <at 1 1 RCIM_Cubic>
          ]}}}
       Scale Curve=1.1
      )
     )
     2=&FloatFromCurve(
      FloatCurve=DataInterface() {propertyValues {Curve {
         defaultValue=1
         keys [
          {arriveTangent=6.8265, interpMode=RCIM_Cubic, leaveTangent=6.8265, tangentMode=RCTM_User, time=0, value=0}
          {interpMode=RCIM_Cubic, tangentMode=RCTM_User, time=0.2, value=1}
          {arriveTangent=0.0797, interpMode=RCIM_Cubic, leaveTangent=0.0797, tangentMode=RCTM_User, time=0.7756, value=0.9869}
          <at 1 0 RCIM_Cubic>
         ]
        }}}
      Scale Curve=&FN_FrontEndFloatModifier(Gameplay=0.5, Front End=0.3)
     )
     3=&FloatFromCurve(
      FloatCurve=DataInterface() {propertyValues {Curve {keys [
          <at 0 0.5 RCIM_Cubic>
          {arriveTangent=2.1543, interpMode=RCIM_Cubic, leaveTangent=2.1543, time=0.1075, value=2.8853}
          {arriveTangentWeight=0.0865, interpMode=RCIM_Cubic, leaveTangentWeight=0.2862, tangentMode=RCTM_User, time=0.2321, value=1}
          {arriveTangentWeight=0.2434, interpMode=RCIM_Cubic, leaveTangentWeight=0.1554, tangentMode=RCTM_User, time=0.985, value=1}
         ]}}}
      Scale Curve=&FN_FrontEndFloatModifier(Gameplay=1, Front End=0.5)
     )
     4=&FloatFromCurve(FloatCurve=DataInterface() {propertyValues {Curve {
         defaultValue=1
         keys [
          {arriveTangent=0.129, interpMode=RCIM_Cubic, leaveTangent=0.129, tangentMode=RCTM_User, time=0, value=0}
          {interpMode=RCIM_Cubic, tangentMode=RCTM_User, time=0.5, value=1}
         ]
        }}})
     5=&FloatFromCurve(FloatCurve=DataInterface() {propertyValues {Curve {
         defaultValue=0
         keys [{interpMode=RCIM_Cubic, tangentMode=RCTM_User, time=0, value=1}, {interpMode=RCIM_Cubic, tangentMode=RCTM_User, time=1, value=0}]
        }}})
     6=&Multiply_Float(
      A=10
      B=&FloatFromCurve(FloatCurve=DataInterface() {propertyValues {Curve {keys [
           {interpMode=RCIM_Cubic, tangentMode=RCTM_User, time=0, value=0}
           {interpMode=RCIM_Cubic, tangentMode=RCTM_User, time=0.1, value=0}
           {interpMode=RCIM_Cubic, tangentMode=RCTM_User, time=0.2, value=0.8}
           {interpMode=RCIM_Cubic, tangentMode=RCTM_User, time=1, value=0}
          ]}}})
     )
     Index=&SpawnGroup()
     Num=::NewEnumerator5
    )
    Scale RGB=&MergeHelper_VectorSwitch(
     0=<1 1 1>
     1=<1 1 1>
     2=&VectorFromFloat(Value=&FloatFromCurve(
       FloatCurve=DataInterface() {propertyValues {Curve {keys [{interpMode=RCIM_Cubic, tangentMode=RCTM_User, time=0, value=1}, {interpMode=RCIM_Cubic, tangentMode=RCTM_User, time=1, value=1}]}}}
       Scale Curve=5
      ))
     3=&VectorFromFloat(Value=&FloatFromCurve(FloatCurve=DataInterface() {propertyValues {Curve {keys [
           {interpMode=RCIM_Cubic, tangentMode=RCTM_User, time=0, value=0.5}
           {arriveTangent=2.7979, interpMode=RCIM_Cubic, leaveTangent=2.7979, tangentMode=RCTM_User, time=0.6286, value=1}
           {arriveTangent=0.8742, interpMode=RCIM_Cubic, leaveTangent=0.8742, time=0.63, value=3}
           {arriveTangent=0.1095, interpMode=RCIM_Cubic, leaveTangent=0.1095, tangentMode=RCTM_User, time=0.6747, value=1.0403}
           {arriveTangentWeight=0.139, interpMode=RCIM_Cubic, leaveTangentWeight=0.0967, tangentMode=RCTM_User, time=0.958, value=1.0156}
           {arriveTangent=50.355, interpMode=RCIM_Cubic, leaveTangent=50.355, tangentMode=RCTM_User, time=1, value=5}
          ]}}}))
     4=<1 1 1>
     5=<1 1 1>
     6=&Multiply_VectorByFloat(
      Vector=&Multiply_VectorByFloat(Vector=<1 1 1>, Float=&FN_FrontEndFloatModifier(Gameplay=1, Front End=0.5))
      Float=&Multiply_Float(
       A=&RandomRangeFloat(Minimum=0.5, Maximum=1, Evaluation Type=::NewEnumerator1)
       B=&Multiply_Float(A=&FNTimeOfDayScalar(Night Scale=0.2, Morning Scale=1, Evening Scale=1), B=10)
      )
     )
     Index=&SpawnGroup()
     Num=::NewEnumerator5
    )
    ScaleA=true
    ScaleRGB=true
   )
   &ScaleSpriteSize(
    Scale Factor=&MergeHelper_Vector2Switch(
     0=&Multiply_Vector2D_ByFloat(
      Float=&FloatFromCurve(FloatCurve=DataInterface() {propertyValues {Curve {keys [
           {arriveTangent=2.7112, interpMode=RCIM_Cubic, leaveTangent=2.7112, tangentMode=RCTM_User, time=0, value=0}
           {interpMode=RCIM_Cubic, tangentMode=RCTM_User, time=0.7, value=1}
           {arriveTangent=-5.8355, tangentMode=RCTM_User, time=1, value=0}
          ]}}})
      Vector2D=<1 1>
     )
     1=&Multiply_Vector2D_ByFloat(
      Float=&FloatFromCurve(FloatCurve=DataInterface() {propertyValues {Curve {
          defaultValue=1
          keys [
           {arriveTangent=2.538, interpMode=RCIM_Cubic, leaveTangent=2.538, tangentMode=RCTM_User, time=0, value=0}
           {arriveTangent=-0.0995, interpMode=RCIM_Cubic, leaveTangent=-0.0995, tangentMode=RCTM_User, time=0.6, value=1.2}
           <at 1 0 RCIM_Cubic>
          ]
         }}})
      Vector2D=<1 1>
     )
     2=&Vector2DFromFloat(Value=&FloatFromCurve(FloatCurve=DataInterface() {propertyValues {Curve {keys [
           {interpMode=RCIM_Cubic, tangentMode=RCTM_User, time=0, value=0.3}
           {interpMode=RCIM_Cubic, tangentMode=RCTM_User, time=0.1, value=1}
           <at 1 0>
          ]}}}))
     3=&Vector2DFromFloat(Value=&FloatFromCurve(
       FloatCurve=DataInterface() {propertyValues {Curve {keys [
           {interpMode=RCIM_Cubic, tangentMode=RCTM_User, time=0, value=0.5}
           {
            arriveTangent=0.9101
            arriveTangentWeight=0.1572
            interpMode=RCIM_Cubic
            leaveTangent=0.9101
            leaveTangentWeight=0.187
            tangentMode=RCTM_User
            time=0.5699
            value=0.7311
           }
           {arriveTangentWeight=0.0486, interpMode=RCIM_Cubic, leaveTangentWeight=0.0063, tangentMode=RCTM_User, time=0.6002, value=0.6814}
           {arriveTangentWeight=0.0022, interpMode=RCIM_Cubic, leaveTangentWeight=0.0079, tangentMode=RCTM_User, time=0.6268, value=0.8592}
           {arriveTangentWeight=0.0079, interpMode=RCIM_Cubic, leaveTangentWeight=0.005, tangentMode=RCTM_User, time=0.648, value=0.7085}
           {arriveTangentWeight=0.0056, interpMode=RCIM_Cubic, leaveTangentWeight=0.0103, tangentMode=RCTM_User, time=0.6633, value=1.0562}
           {arriveTangentWeight=0.0098, interpMode=RCIM_Cubic, leaveTangentWeight=0.0053, tangentMode=RCTM_User, time=0.6859, value=0.8333}
           {arriveTangentWeight=0.0153, interpMode=RCIM_Cubic, leaveTangentWeight=0.0391, tangentMode=RCTM_User, time=0.6943, value=1.2148}
           {arriveTangentWeight=0.2165, interpMode=RCIM_Cubic, leaveTangentWeight=0.1618, tangentMode=RCTM_User, time=0.7165, value=1.328}
           {arriveTangentWeight=0.1546, interpMode=RCIM_Cubic, leaveTangentWeight=0.0794, tangentMode=RCTM_User, time=0.7285, value=0.9473}
           {arriveTangentWeight=0.0037, interpMode=RCIM_Cubic, leaveTangentWeight=0.003, tangentMode=RCTM_User, time=0.7458, value=1.2876}
           {arriveTangentWeight=0.003, interpMode=RCIM_Cubic, leaveTangentWeight=0.0044, tangentMode=RCTM_User, time=0.7578, value=0.9119}
           {arriveTangentWeight=0.0044, interpMode=RCIM_Cubic, leaveTangentWeight=0.0039, tangentMode=RCTM_User, time=0.7667, value=1.849}
           {arriveTangentWeight=0.0039, interpMode=RCIM_Cubic, leaveTangentWeight=0.0028, tangentMode=RCTM_User, time=0.7769, value=1.285}
           {arriveTangentWeight=0.0028, interpMode=RCIM_Cubic, leaveTangentWeight=0.0048, tangentMode=RCTM_User, time=0.7867, value=1.7376}
           {arriveTangentWeight=0.0048, interpMode=RCIM_Cubic, leaveTangentWeight=0.0024, tangentMode=RCTM_User, time=0.8024, value=1.4407}
           {arriveTangentWeight=0.146, interpMode=RCIM_Cubic, leaveTangentWeight=0.1771, tangentMode=RCTM_User, time=0.8172, value=1.9071}
           {arriveTangentWeight=0.0029, interpMode=RCIM_Cubic, leaveTangentWeight=0.0043, tangentMode=RCTM_User, time=0.8354, value=1.3099}
           {arriveTangentWeight=0.014, interpMode=RCIM_Cubic, leaveTangentWeight=0.0241, tangentMode=RCTM_User, time=0.8463, value=1.9292}
           {arriveTangent=12.2531, interpMode=RCIM_Cubic, leaveTangent=12.2531, time=0.8611, value=1.6606}
           {arriveTangentWeight=0.0505, interpMode=RCIM_Cubic, leaveTangentWeight=0.0235, tangentMode=RCTM_User, time=0.8719, value=2.2433}
           {arriveTangentWeight=0.0186, interpMode=RCIM_Cubic, leaveTangentWeight=0.0293, tangentMode=RCTM_User, time=0.8848, value=0.896}
           {arriveTangentWeight=0.0007, interpMode=RCIM_Cubic, leaveTangentWeight=0.1139, tangentMode=RCTM_User, time=0.8936, value=2.2846}
           {arriveTangentWeight=0.0104, interpMode=RCIM_Cubic, leaveTangentWeight=0.1242, tangentMode=RCTM_User, time=0.9113, value=1.7885}
           {
            arriveTangent=0.5964
            arriveTangentWeight=0.0737
            interpMode=RCIM_Cubic
            leaveTangent=0.5964
            leaveTangentWeight=0.0932
            tangentMode=RCTM_User
            time=0.92
            value=2.2798
           }
           {arriveTangent=-13.6611, interpMode=RCIM_Cubic, leaveTangent=-13.6611, time=0.9307, value=1.0319}
           {arriveTangent=-14.8922, interpMode=RCIM_Cubic, leaveTangent=-14.8922, time=0.9495, value=1.8774}
           {arriveTangentWeight=0.6437, interpMode=RCIM_Cubic, tangentMode=RCTM_User, time=1, value=0}
          ]}}}
       Scale Curve=0.5
      ))
     4=<0 0>
     Index=&SpawnGroup()
     Num=::NewEnumerator3
    )
    Scale Sprite Size Mode=::NewEnumerator2
   )
   /SetVariables(Particles.BurstWaveSize:Vector2f=&Multiply_Vector2D_ByFloat(
     Float=&FloatFromCurve(FloatCurve=DataInterface() {propertyValues {Curve {keys [
          {interpMode=RCIM_Cubic, tangentMode=RCTM_User, time=0, value=1}
          {interpMode=RCIM_Cubic, tangentMode=RCTM_User, time=0.7047, value=1.7}
          {interpMode=RCIM_Cubic, tangentMode=RCTM_User, time=1, value=1}
         ]}}})
     Vector2D=$Particles.SpriteSize
    )) {isSetParameters=true}
   /SetVariables(Particles.BurstWaveRaysSize:Vector2f=&Multiply_Vector2D_ByFloat(Float=1.1, Vector2D=$Particles.BurstWaveSize)) {isSetParameters=true}
   &ScaleMeshSize(Scale Factor=&MergeHelper_VectorSwitch(
     0=<0 0 0>
     1=<0 0 0>
     2=<1 1 1>
     3=<0 0 0>
     4=&VectorFromFloat(Value=&FloatFromCurve(FloatCurve=DataInterface() {propertyValues {Curve {
          defaultValue=1
          keys [
           {arriveTangent=3.5237, interpMode=RCIM_Cubic, leaveTangent=3.5237, tangentMode=RCTM_User, time=0, value=0}
           {interpMode=RCIM_Cubic, tangentMode=RCTM_User, time=0.9682, value=1}
           {interpMode=RCIM_Cubic, tangentMode=RCTM_User, time=1, value=0}
          ]
         }}}))
     5=&VectorFromFloat(Value=&FloatFromCurve(FloatCurve=DataInterface() {propertyValues {Curve {
          defaultValue=1
          keys [{interpMode=RCIM_Cubic, tangentMode=RCTM_User, time=0.9682, value=1}, {interpMode=RCIM_Cubic, tangentMode=RCTM_User, time=1, value=0}]
         }}}))
     6=&VectorFromCurve(VectorCurve=DataInterface() {propertyValues {
        XCurve {keys [<at 0.2848 1 RCIM_Cubic>, {arriveTangent=-3.0417, interpMode=RCIM_Cubic, leaveTangent=-3.0417, tangentMode=RCTM_User, time=1, value=0}]}
        YCurve {keys [<at 0 1>,<at 1 1>]}
        ZCurve {keys [<at 0 1>,<at 1 1>]}
       }})
     Index=&SpawnGroup()
     Num=::NewEnumerator5
    ))
   &FN_Sprite_Attributes(
    Color=&MultiplyLinearColor(
     Linear Color=&MergeHelper_ColorSwitch(
      0=<rgba 1.2 0.1281 0 1>
      1=<rgba 0 0.2487 1.2 1>
      2=<rgba 1.2 0 0.0069 1>
      3=?
      Index=&MakeIntFromFloat(Float=$User.Variant, Conversion Mode=::NewEnumerator1)
      Num=::NewEnumerator1
     )
     Scale Factor=&MakeVector4(
      X=1
      Y=1
      Z=1
      W=&MaskFloatBySpawnGroup(
       Masked Float=&FloatFromCurve(FloatCurve=DataInterface() {propertyValues {Curve {
           defaultValue=1
           keys [
            {arriveTangent=4.8699, interpMode=RCIM_Cubic, leaveTangent=4.8699, tangentMode=RCTM_User, time=0, value=0.5}
            {interpMode=RCIM_Cubic, tangentMode=RCTM_User, time=0.2, value=1}
           ]
          }}})
       Passthrough Float=&FloatFromCurve(FloatCurve=DataInterface() {propertyValues {Curve {
           defaultValue=1
           keys [
            {interpMode=RCIM_Cubic, tangentMode=RCTM_User, time=0, value=0}
            {interpMode=RCIM_Cubic, tangentMode=RCTM_User, time=0.01, value=1}
            {arriveTangentWeight=0.3461, interpMode=RCIM_Cubic, leaveTangentWeight=0.0814, tangentMode=RCTM_User, time=0.4257, value=1}
            <at 0.8293 1 RCIM_Cubic>
           ]
          }}})
      )
     )
    )
    Create Color Attribute=true
    Create SpriteSize Attribute=true
    SpriteSize=&Multiply_Vector2D(
     A=&Vector2DFromFloat(Value=&FloatFromCurve(
       FloatCurve=DataInterface() {propertyValues {Curve {keys [
           {arriveTangent=1.7893, interpMode=RCIM_Cubic, leaveTangent=1.7893, tangentMode=RCTM_User, time=0, value=0.3}
           <at 0.5 1>
           {interpMode=RCIM_Cubic, tangentMode=RCTM_User, time=0.91, value=1}
           <at 1 0>
          ]}}}
       Scale Curve=150
      ))
     B=&Vector2DFromFloat(Value=&Multiply_Float(
       A=&MaskFloat(
        Mask Source=0
        Mask Parameter=?
        Mask=0
        Masked Float=&FloatFromCurve(
         FloatCurve=DataInterface() {propertyValues {Curve {keys [<at 0 0>, <at 0.05 1>, {interpMode=RCIM_Cubic, tangentMode=RCTM_User, time=0.83, value=1}, <at 0.9 0>]}}}
         Scale Curve=0.7
        )
        Passthrough Float=&FloatFromCurve(FloatCurve=DataInterface() {propertyValues {Curve {keys [
             {arriveTangent=6.2432, interpMode=RCIM_Cubic, leaveTangent=6.2432, tangentMode=RCTM_User, time=0, value=0}
             {arriveTangentWeight=0.1385, interpMode=RCIM_Cubic, leaveTangentWeight=0.2947, tangentMode=RCTM_User, time=0.2592, value=1}
            ]}}})
       )
       B=&Multiply_Float(
        A=1
        B=&FN_LOD_Distance_Remap(Start Distance=350, End Distance=380, Minimum Output=0.7, Maximum Output=1, Clamp Output=true)
       )
      ))
    )
    Write to Module Namespace=true
   )
   &Light_Attributes(
    Color=1
    Color Value To Scale:LinearColor=&Multiply_LinearColorByFloat(
     Linear Color=<rgba 3 1.3939 0 1>
     Float=&FloatFromCurve(FloatCurve=DataInterface() {propertyValues {Curve {keys [
          {interpMode=RCIM_Cubic, tangentMode=RCTM_User, time=0.6, value=1}
          {interpMode=RCIM_Cubic, tangentMode=RCTM_User, time=0.65, value=5}
          {arriveTangent=5.1514, interpMode=RCIM_Cubic, leaveTangent=5.1514, tangentMode=RCTM_User, time=0.7028, value=1}
          {interpMode=RCIM_Cubic, tangentMode=RCTM_User, time=0.93, value=1.9412}
          {interpMode=RCIM_Cubic, tangentMode=RCTM_User, time=0.95, value=3}
          {interpMode=RCIM_Cubic, tangentMode=RCTM_User, time=1, value=1}
         ]}}})
    )
    Light Radius:NiagaraFloat=&FloatFromCurve(
     FloatCurve=DataInterface() {propertyValues {Curve {keys [{interpMode=RCIM_Cubic, tangentMode=RCTM_User, time=3, value=0.5}, <at 3.4 1>]}}}
     CurveIndex=$Emitter.Age
     Scale Curve=125
    )
    Radius=1
    ScaleA=true
    ScaleRGB=false
   )
   /SetVariables(
    Particles.WaveColor:LinearColor=&Multiply_LinearColorByFloat(
     Linear Color=$Particles.Initial.Color
     Float=&FloatFromCurve(FloatCurve=DataInterface() {propertyValues {Curve {keys [{interpMode=RCIM_Cubic, tangentMode=RCTM_User, time=0, value=1}, <at 1 0>]}}})
    )
    Particles.WaveScale:Vector3f=&Multiply_VectorByFloat(
     Vector=$Particles.Scale
     Float=&FloatFromCurve(
      FloatCurve=DataInterface() {propertyValues {Curve {
         defaultValue=1
         keys [
          {arriveTangent=1.4958, interpMode=RCIM_Cubic, leaveTangent=1.4958, tangentMode=RCTM_User, time=0, value=0.1}
          {interpMode=RCIM_Cubic, tangentMode=RCTM_User, time=1, value=1}
         ]
        }}}
      Scale Curve=1.3
     )
    )
   ) {isSetParameters=true}
  ]
  rendererClasses [
   &NiagaraSpriteRendererProperties
   &NiagaraMeshRendererProperties
   &NiagaraLightRendererProperties
  ]
  renderers [
   {rendererClass=&NiagaraSpriteRendererProperties}
   {rendererClass=&NiagaraSpriteRendererProperties, rendererIndex=1}
   {rendererClass=&NiagaraSpriteRendererProperties, rendererIndex=2}
   {rendererClass=&NiagaraSpriteRendererProperties, rendererIndex=3}
   {rendererClass=&NiagaraMeshRendererProperties, rendererIndex=4}
   {rendererClass=&NiagaraSpriteRendererProperties, rendererIndex=5}
   {rendererClass=&NiagaraSpriteRendererProperties, rendererIndex=6}
   {rendererClass=&NiagaraLightRendererProperties, rendererIndex=7}
   {rendererClass=&NiagaraMeshRendererProperties, rendererIndex=8}
  ]
 }
 Glitch {
  emitterUpdateScript [
   &EmitterState(Life Cycle Mode=::NewEnumerator1, Loop Behavior=::NewEnumerator1, Loop Duration=10, UseLoopDelay=false)
   &SpawnBurst_Instantaneous(Spawn Count=7, Use Spawn Probability=false, UseLoopCountLimit=false) {module=SpawnBurst_Instantaneous001}
  ]
  particleSpawnScript [
   /SetVariables(Particles.GlitchColor:LinearColor=<rgba 5 1.0838 0 1>, Particles.GlitchSize:Vector2f=<5 30>) {isSetParameters=true}
   &InitializeParticle(
    Color=<rgba 1 0.5825 0 1>
    Lifetime=&SelectFloatFromArray(
     Direct Array Index=&SpawnGroup()
     Float Selection Array=DataInterface() {propertyValues {FloatData [0.5,0.5,0.6,0.3]}}
    )
    Mesh Scale Mode=::NewEnumerator3
    Mesh Uniform Scale=0.8
    Position Offset=&MergeHelper_VectorSwitch(0=<0 0 0>, 1=<0 0 0>, 2=<-20 0 -10>, 3=<-20 0 -10>, Index=&SpawnGroup(), Num=::NewEnumerator2)
    Sprite Size Mode=::NewEnumerator3
    Uniform Sprite Size=40
    UsePositionOffset=true
    Write Material Random=true
   )
   &DynamicMaterialParameters(
    Index 0 Param 1=&RandomRangeFloat(Minimum=0, Maximum=1)
    Index 0 Param 2=&RandomRangeFloat(Minimum=8, Maximum=12)
    Index 0 Param 3=&FloatFromCurve(FloatCurve=DataInterface() {propertyValues {Curve {keys [{interpMode=RCIM_Cubic, tangentMode=RCTM_User, time=0, value=1.6}, {arriveTangent=7.3367, tangentMode=RCTM_User, time=1, value=5}]}}})
    Module.Index0Float0Write=true
    Module.Index0Float1Write=true
    Module.Index0Float2Write=true
    Module.Index0Float3Write=true
   )
   /Game/Effects/Niagara/Modules/Sprite/FN_SpriteUV(End Frame=3, Write to Module Namepsace=true)
   &CameraOffset(Camera Offset Amount=-10)
   &ShapeLocation(
    MaskSpawnGroup=false
    Override Seed=false
    Ring Radius=&MaskFloatBySpawnGroup(Masked Float=10, Passthrough Float=0)
    Shape Primitive=::NewEnumerator4
    UseOwnerScale=true
   )
  ]
  particleUpdateScript [
   &ParticleState()
   &ScaleSpriteSize(
    Uniform Curve Scale=$System.FE_Scale
    Uniform Curve Sprite Scale=DataInterface() {propertyValues {Curve {defaultValue=0, keys [{arriveTangent=3.3458, interpMode=RCIM_Cubic, leaveTangent=3.3458, tangentMode=RCTM_User, time=0, value=1}, <at 1 3>]}}}
   )
   &ScaleColor(
    Scale Alpha=&FloatFromCurve(FloatCurve=DataInterface() {propertyValues {Curve {
        defaultValue=1
        keys [
         {arriveTangent=0.3137, interpMode=RCIM_Cubic, leaveTangent=0.3137, tangentMode=RCTM_User, time=0, value=1}
         {arriveTangent=-2.0958, arriveTangentWeight=0.3333, interpMode=RCIM_Cubic, leaveTangent=-2.0958, tangentMode=RCTM_User, time=1, value=0}
        ]
       }}})
    Scale RGB=&VectorFromCurve(VectorCurve=DataInterface() {propertyValues {
       XCurve {defaultValue=1, keys [{interpMode=RCIM_Cubic, tangentMode=RCTM_User, time=0, value=1}, <at 1 0>]}
       YCurve {defaultValue=1, keys [{interpMode=RCIM_Cubic, tangentMode=RCTM_User, time=0, value=1}, <at 1 0>]}
       ZCurve {defaultValue=1, keys [{interpMode=RCIM_Cubic, tangentMode=RCTM_User, time=0, value=1}, <at 1 0>]}
      }})
    ScaleA=true
    ScaleRGB=true
   )
   &SolveForcesAndVelocity(Clamp Velocity=false, Limit Acceleration=false)
   /SetVariables(Particles.Glitch DynP:Vector4f=&MakeVector4(
     X=&RandomRangeFloat(Minimum=0, Maximum=1)
     Y=&Lerp_Float(A=0.5, B=1, Alpha=$Particles.NormalizedAge)
     Z=/Niagara/DynamicInputs/Integer/V2/Round_Float(Float:NiagaraFloat=&FloatFromCurve(
       FloatCurve=DataInterface() {propertyValues {Curve {defaultValue=1, keys [<at 0 1>,<at 0.2035 0>,<at 0.444 0.9782>,<at 0.7 0>,<at 1 0.25>]}}}
       Scale Curve=4
      ))
     W=-1
    )) {isSetParameters=true}
  ]
  rendererClasses [&NiagaraSpriteRendererProperties]
  renderers [{rendererClass=&NiagaraSpriteRendererProperties}]
 }
 Lightning {
  emitterUpdateScript [
   &EmitterState(Life Cycle Mode=::NewEnumerator1, Loop Behavior=::NewEnumerator1, Loop Duration=10, UseLoopDelay=false)
   &SpawnBurst_Instantaneous(Spawn Count=5, Spawn Time=3.4, Use Spawn Probability=false, UseLoopCountLimit=false)
   &SpawnRate(
    SpawnRate=&FloatFromCurve(FloatCurve=DataInterface() {propertyValues {Curve {keys [<at 1 0 RCIM_Cubic>,<at 3.3 1>,<at 4 0>]}}}, CurveIndex=$Emitter.Age, Scale Curve=3)
    Use Spawn Probability=false
   ) {module=SpawnRate001}
   &SpawnBurst_Instantaneous(Spawn Count=3, Spawn Group=1, Spawn Time=0.5, Use Spawn Probability=false, UseLoopCountLimit=false) {module=SpawnBurst_Instantaneous001}
   &SpawnBurst_Instantaneous(Spawn Count=3, Spawn Group=1, Spawn Time=3.4, Use Spawn Probability=false, UseLoopCountLimit=false) {module=SpawnBurst_Instantaneous002}
   &SpawnRate(
    SpawnGroup=1
    SpawnRate=&FloatFromCurve(
     FloatCurve=DataInterface() {propertyValues {Curve {
        defaultValue=1
        keys [
         <at 0 0 RCIM_Cubic>
         <at 0.1196 0.6932 RCIM_Cubic>
         {arriveTangentWeight=0.0646, interpMode=RCIM_Constant, time=0.4174, value=0}
         <at 0.4194 10>
         {arriveTangentWeight=0.086, interpMode=RCIM_Constant, time=0.5107, value=0}
         <at 0.7026 8 RCIM_Cubic>
         {interpMode=RCIM_Cubic, leaveTangent=40, time=0.7526, value=0}
        ]
       }}}
     CurveIndex=$Emitter.NormalizedLoopAge
    )
    Use Spawn Probability=false
   )
  ]
  particleSpawnScript [
   /SetVariables(
    Particles.LightningIndex:NiagaraInt32=&RandomRangeInt(Minimum=0, Maximum=3)
    Particles.LightningLength:NiagaraFloat=&Multiply_Float(
     A=/Niagara/DynamicInputs/Helpers/DistributionCurveFloat(
      Minimum:NiagaraFloat=0.4207
      Maximum:NiagaraFloat=0.744
      Distribution Curve:NiagaraDataInterfaceCurve=DataInterface() {propertyValues {Curve {defaultValue=0, keys [<at 0.3 0 RCIM_Constant>, {tangentMode=RCTM_User, time=0.85, value=1}]}}}
      Curve Parameter Type:NiagaraInt32=1
      "Curve Parameter Type:":NiagaraInt32=0
      Curve Parameter:NiagaraFloat=?
      Evaluation Type:ENiagaraRandomnessEvaluation=::NewEnumerator0
      Recalculate Random Each Loop:NiagaraBool=true
      Randomness Mode:ENiagaraRandomnessMode=::NewEnumerator0
      Random Seed:NiagaraInt32=?
      Fixed Random Seed:NiagaraBool=false
     )
     B=&FloatFromCurve(
      FloatCurve=DataInterface() {propertyValues {Curve {
         defaultValue=1
         keys [{interpMode=RCIM_Cubic, tangentMode=RCTM_User, time=1, value=0.3}, {interpMode=RCIM_Cubic, tangentMode=RCTM_User, time=3.4, value=1}]
        }}}
      CurveIndex=$Emitter.Age
     )
    )
    Particles.MirrorLightning:NiagaraBool=&RandomBool()
    Particles.Random=&RandomRangeFloat(Minimum=0, Maximum=1)
    Particles.VisibilityTag=&SpawnGroup()
   ) {isSetParameters=true}
   &PartitionParticles(Number Of Partitions=3, Partition Mode:ENiagara_PartitionMode_v2=::NewEnumerator5)
   &InitializeParticle(
    AdjustAlpha=false
    AdjustHue=false
    AdjustSaturation=true
    AdjustValue=true
    Color=<rgba 2 0.1311 0 1>
    Color Mode=::NewEnumerator3
    Lifetime=&RandomRangeFloat(Minimum=0.2, Maximum=0.3)
    Position Offset=$System.BurstPosition
    Saturation Range:Vector2f=<0.95 1>
    UsePositionOffset=true
    Value Range:Vector2f=<1 2>
    Write Material Random=true
   )
   &ShapeLocation(MaskSpawnGroup=false, Override Seed=false, Sphere Radius=1, UseOwnerScale=true)
   &InitialMeshOrientation(
    Mesh Orientation Mode=::NewEnumerator2
    Orientation Vector=&FN_FrontEndVector3Modifier(Gameplay=<-1 0 0>, Front End=&RandomVector(Vector Scale=1))
    Rotation=&MakeVector(
     X=&MakeCustomFloatFromBool(
      Bool=&SetBoolByFloatComparison(A=$Emitter.Age, B=3.4)
      True Float=&Add_Float(
       A=&FloatFromCurve(
        FloatCurve=DataInterface() {propertyValues {Curve {keys [<at 0 1>,<at 1 0>]}}}
        CurveIndex=&ReturnNormalizedExecIndex()
       )
       B=&RandomRangeFloat(Minimum=-0.1, Maximum=0.1)
      )
      False Float=&RandomRangeFloat(Minimum=0, Maximum=1)
     )
     Y=&RandomRangeFloat(Minimum=-0.3, Maximum=0.3)
     Z=0
    )
    Use Rotation Vector=true
   )
   &CameraOffset(Camera Offset Amount=5)
  ]
  particleUpdateScript [
   &ParticleState()
   &FN_Mesh_Attributes(
    Create Scale Attribute=true
    Scale=&Multiply_Vector(
     A=&Multiply_VectorByFloat(
      Vector=&MakeCustomVectorFromBool(Bool=$Particles.Initial.MirrorLightning, True Vector=<1 -1 -1>, False Vector=<1 1 1>)
      Float=$Particles.Initial.LightningLength
     )
     B=&Multiply_VectorByFloat(
      Vector=&MakeVector(
       X=&FloatFromCurve(
        FloatCurve=DataInterface() {propertyValues {Curve {
           defaultValue=1
           keys [
            {arriveTangent=0.8401, interpMode=RCIM_Cubic, leaveTangent=0.8401, tangentMode=RCTM_User, time=0, value=0.8}
            {arriveTangent=0.1, interpMode=RCIM_Cubic, leaveTangent=0.1, time=0.1317, value=1.0137}
            {interpMode=RCIM_Cubic, tangentMode=RCTM_User, time=1, value=0.9}
           ]
          }}}
        Scale Curve=0.7
       )
       Y=1
       Z=1
      )
      Float=$System.FE_Scale
     )
    )
   )
   /SetVariables(
    Particles.DynamicParametersAlt:Vector4f=&MakeVector4(
     X=&Multiply_Float(
      A=$Particles.Initial.LightningLength
      B=&FloatFromCurve(
       FloatCurve=DataInterface() {propertyValues {Curve {keys [{interpMode=RCIM_Cubic, tangentMode=RCTM_User, time=0, value=1}, <at 1 0.3>]}}}
       Scale Curve=&RandomRangeFloat(Minimum=2, Maximum=3)
      )
     )
     Y=&UniformAOrBFloat(A=0.5, B=1.5, Evaluation Type=::NewEnumerator0, Recalculate Random Each Loop=true, Override Seed=false)
     Z=&FloatFromCurve(FloatCurve=DataInterface() {propertyValues {Curve {
         defaultValue=1
         keys [
          {arriveTangent=-2.478, interpMode=RCIM_Cubic, leaveTangent=-2.478, tangentMode=RCTM_User, time=0, value=1}
          {interpMode=RCIM_Cubic, tangentMode=RCTM_User, time=1, value=0}
         ]
        }}})
     W=0
    )
    Particles.DynamicParametersAlt2:Vector4f=&MakeVector4(X=1, Y=1, Z=0, W=0)
   ) {isSetParameters=true}
   &FN_Mesh_Attributes(
    Color=&MultiplyLinearColor(
     Linear Color=&RandomRangeLinearColor(Minimum=<rgba 5 1 5 1>, Maximum=<rgba 5 1 3.6667 1>)
     Scale Factor=&MakeVector4FromVectorAndFloat(
      "Vector (RGB)"=&VectorFromFloat(Value=&RandomRangeFloat(Minimum=0, Maximum=2, Evaluation Type=::NewEnumerator1))
      "Float (Alpha)"=&FloatFromCurve(FloatCurve=DataInterface() {propertyValues {Curve {keys [<at 0.7 1>,<at 1 0>]}}})
     )
    )
    Create Color Attribute=true
    Create Orientation Attribute=true
    Create Scale Attribute=true
    Orientation:Quat4f=&MakeQuaternion(
     Quaternion From=1
     Axis=&RandomRangeVector(Minimum=<-0.3 -0.3 -1>, Maximum=<0.3 0.3 1>)
     Angle=&RandomRangeFloat(Minimum=0, Maximum=360)
     Yaw=&RandomRangeFloat(Minimum=-0.1, Maximum=0.1)
     Pitch=&RandomRangeFloat(Minimum=0.05, Maximum=0.45)
     Roll=&RandomRangeFloat(Minimum=-1, Maximum=1)
     From=?
     To=?
     Angle Type=::NewEnumerator1
     Coordinate Space=::Local
     Rotation Matrix=Linked() {linkedVariable {name=Engine.Owner.SystemLocalToWorldNoScale, type=NiagaraInt32, defaultValue=~, description=~}}
     Normalize Vectors=false
    )
    Scale=&Multiply_VectorByFloat(
     Vector=&VectorFromFloat(Value=&Multiply_Float(
       A=&Multiply_Float(A=0.3, B=0.6)
       B=&FloatFromCurve(
        FloatCurve=DataInterface() {propertyValues {Curve {keys [<at 0 1>,<at 0.3653 1>,<at 0.4783 1.3>,<at 1 1.3>]}}}
        CurveIndex=$Emitter.NormalizedLoopAge
       )
      ))
     Float=&FloatFromCurve(
      FloatCurve=DataInterface() {propertyValues {Curve {keys [
          {arriveTangent=3.2151, interpMode=RCIM_Cubic, leaveTangent=3.2151, tangentMode=RCTM_User, time=0, value=0.6}
          {arriveTangent=0.4, interpMode=RCIM_Cubic, leaveTangent=0.4, time=0.5973, value=0.9154}
          {arriveTangent=0.0224, interpMode=RCIM_Cubic, leaveTangent=0.0224, tangentMode=RCTM_User, time=1, value=1}
         ]}}}
      Scale Curve=$System.FE_Scale
     )
    )
    Write to Module Namespace=true
   ) {module=FN_Mesh_Attributes002}
   /SetVariables(
    Particles.DynamicParametersAlt1:Vector4f=&MakeVector4(X=0, Y=0, Z=0, W=13)
    Particles.DynamicParametersAlt3:Vector4f=&MakeVector4(
     X=&FloatFromCurve(FloatCurve=DataInterface() {propertyValues {Curve {keys [<at 0 0.5>,<at 1 1>]}}})
     Y=3
     Z=&FloatFromCurve(FloatCurve=DataInterface() {propertyValues {Curve {keys [{interpMode=RCIM_Cubic, tangentMode=RCTM_Break, time=0.5, value=-0.5}, <at 1 0.1>]}}})
     W=3
    )
    Particles.DynamicParametersAlt4:Vector4f=&MakeVector4(
     X=&Multiply_Float(
      A=&FloatFromCurve(FloatCurve=DataInterface() {propertyValues {Curve {keys [{interpMode=RCIM_Cubic, leaveTangent=8.1085, tangentMode=RCTM_Break, time=0.8568, value=1}, {tangentMode=RCTM_User, time=1, value=0}]}}})
      B=&Divide_Float(
       A=&MakeFloatFromVector(Vector=$Particles.FN_Mesh_Attributes002.Scale, Channel=::NewEnumerator0)
       B=&RandomRangeFloat(Minimum=50, Maximum=100)
      )
     )
     Y=&FloatFromCurve(
      FloatCurve=DataInterface() {propertyValues {Curve {keys [
          {interpMode=RCIM_Cubic, tangentMode=RCTM_Break, time=0.7, value=0}
          {arriveTangent=-0.0859, interpMode=RCIM_Cubic, tangentMode=RCTM_Break, time=1, value=-1}
         ]}}}
      Scale Curve=50
     )
     Z=&Waveform(
      Global Input Position=&RandomRangeFloat(Minimum=0, Maximum=15)
      Global Phase=$Particles.Age
      "Waveform [1]"=::NewEnumerator12
      "[1] Phase Scale"=2
      "[1] Frequency"=1
      "[1] Amplitude Min/Max"=<0 1>
     )
     W=25
    )
    Particles.DynamicParametersAlt5:Vector4f=&MakeVector4(
     X=&FloatFromCurve(
      FloatCurve=DataInterface() {propertyValues {Curve {
         defaultValue=1
         keys [
          {arriveTangent=2.157, interpMode=RCIM_Cubic, leaveTangent=2.157, tangentMode=RCTM_User, time=0, value=0}
          {interpMode=RCIM_Cubic, tangentMode=RCTM_User, time=1, value=1}
         ]
        }}}
      Scale Curve=10
     )
     Y=0.1
     Z=&FloatFromCurve(FloatCurve=DataInterface() {propertyValues {Curve {keys [
          {interpMode=RCIM_Cubic, leaveTangent=1.1034, tangentMode=RCTM_Break, time=0, value=-1}
          {arriveTangent=2.2399, tangentMode=RCTM_User, time=0.5, value=0.125}
         ]}}})
     W=&FloatFromCurve(FloatCurve=DataInterface() {propertyValues {Curve {keys [
          {interpMode=RCIM_Cubic, leaveTangent=1.8089, tangentMode=RCTM_Break, time=0.3, value=-0.5}
          {arriveTangent=2.0757, interpMode=RCIM_Cubic, tangentMode=RCTM_Break, time=0.9, value=1}
         ]}}})
    )
   ) {isSetParameters=true}
   &ScaleColor(
    Scale Alpha=&Multiply_Float(
     A=10
     B=&FloatFromCurve(FloatCurve=DataInterface() {propertyValues {Curve {keys [<at 0 0>,<at 0.1 0>,<at 0.2 0.8>,<at 1 0>]}}})
    )
    Scale RGB=&Multiply_VectorByFloat(
     Vector=&Multiply_VectorByFloat(Vector=<1 1 1>, Float=&FN_FrontEndFloatModifier(Gameplay=0.5, Front End=0.5))
     Float=&Multiply_Float(
      A=&RandomRangeFloat(Minimum=0.35, Maximum=1, Evaluation Type=::NewEnumerator1)
      B=&Multiply_Float(A=&FNTimeOfDayScalar(Night Scale=0.2, Morning Scale=1, Evening Scale=1), B=10)
     )
    )
    ScaleA=true
    ScaleRGB=true
   )
   &DynamicMaterialParameters(
    Index 0 Param 1=1
    Index 0 Param 2=0.7
    Index 0 Param 3=&FloatFromCurve(FloatCurve=DataInterface() {propertyValues {Curve {keys [
         {arriveTangent=4.0072, interpMode=RCIM_Cubic, leaveTangent=4.0072, tangentMode=RCTM_User, time=0, value=0}
         {arriveTangent=2.7083, interpMode=RCIM_Cubic, leaveTangent=2.7083, time=0.2543, value=0.6225}
         {arriveTangent=1.8473, interpMode=RCIM_Cubic, leaveTangent=1.8473, time=0.6, value=1.625}
         {arriveTangent=1.6993, interpMode=RCIM_Cubic, leaveTangent=1.6993, tangentMode=RCTM_User, time=1, value=2}
        ]}}})
    Index 0 Param 4=13
    Index 1 Param 1=&FN_FrontEndFloatModifier(Gameplay=1.2, Front End=0.5)
    Index 1 Param 2=$Particles.Initial.Random
    Index 1 Param 3=&FloatFromCurve(FloatCurve=DataInterface() {propertyValues {Curve {keys [<at 0 0.1>,<at 1 1.5>]}}})
    Index 1 Param 4=1
    Index 2 Param 1=&FloatFromCurve(FloatCurve=DataInterface() {propertyValues {Curve {keys [<at 0 1>,<at 1 0.3>]}}})
    Module.Index0Float0Write=true
    Module.Index0Float1Write=true
    Module.Index0Float2Write=true
    Module.Index0Float3Write=true
    Module.Index1Float0Write=true
    Module.Index1Float1Write=true
    Module.Index1Float2Write=true
    Module.Index1Float3Write=true
    Module.Index2Float0Write=true
    Module.Index2Float1Write=true
    Module.Index2Float2Write=true
    Module.Index2Float3Write=true
    Write Parameter Index 1=::NewEnumerator1
    Write Parameter Index 2=::NewEnumerator1
   )
   /SetVariables(Particles.Lightningglowcolor:LinearColor=&MakeLinearColorFromVectorAndFloat(
     "Vector (RGB)"=&MakeVectorFromLinearColorRGB(Color=&MultiplyLinearColor(
       Linear Color=<rgba 5 2.6797 0.3594 1>
       Scale Factor=&MakeVector4_FromFloat(Float=&RandomRangeFloat(Minimum=0.6, Maximum=0.8, Evaluation Type=::NewEnumerator1))
      ))
     "Float (Alpha)"=&Multiply_Float(
      A=&FloatFromCurve(FloatCurve=DataInterface() {propertyValues {Curve {keys [<at 0 1 RCIM_Cubic>, {interpMode=RCIM_Cubic, leaveTangent=-3.6307, time=0.9, value=0}]}}})
      B=1
     )
    )) {isSetParameters=true}
   /SetVariables(Particles.Glow Dy02:Vector4f=&MakeVector4(
     X=&FloatFromCurve(FloatCurve=DataInterface() {propertyValues {Curve {keys [{tangentMode=RCTM_User, time=0, value=0.2}, {tangentMode=RCTM_User, time=0.5, value=1}]}}})
     Y=0.7
     Z=0
     W=12
    )) {isSetParameters=true}
   /SetVariables(Particles.Glow Dy03:Vector4f=&MakeVector4(
     X=&FloatFromCurve(
      FloatCurve=DataInterface() {propertyValues {Curve {keys [<at 0 1>,<at 1 0.5>]}}}
      Scale Curve=&FN_FrontEndFloatModifier(Gameplay=1.5, Front End=1)
     )
     Y=$Particles.Initial.Random
     Z=1
     W=1
    )) {isSetParameters=true}
  ]
  rendererClasses [&NiagaraMeshRendererProperties]
  renderers [
   {rendererClass=&NiagaraMeshRendererProperties, rendererIndex=0}
   {rendererClass=&NiagaraMeshRendererProperties, rendererIndex=1}
   {rendererClass=&NiagaraMeshRendererProperties, rendererIndex=2}
  ]
 }
}
summary {
 systemName=NS_CosmoHound_Emote_Sequence
 emitters [
  {emitterName=Glitch, rendererClasses [&NiagaraSpriteRendererProperties]}
  {
   emitterName=Dark_Glow
   rendererClasses [
    &NiagaraSpriteRendererProperties
    &NiagaraMeshRendererProperties
    &NiagaraLightRendererProperties
   ]
  }
  {emitterName=Lightning, rendererClasses [&NiagaraMeshRendererProperties]}
 ]
 userVariables [
  User.bIsFrontEnd {type=NiagaraBool}
  User.bIsFrontEndPreview {type=NiagaraBool}
  User.Variant {type=NiagaraFloat, defaultValue=0, description=""}
 ]
}
