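# Niagara effect for the CosmoHound cosmetic companion's emote sequence collapse, a short orange pop with flare, rays and glow.
# A single Pop emitter fires two staggered bursts, using spawn group for lifetime and visibility tag.
# Particle update derives separate curve-driven size, color and glitch attributes to feed layered sprite, mesh and light renderers.
# System FE_Scale scales several layers, and the cube-collapse color varies with time of day.
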NiagaraSystem /CosmeticCompanions/FX/Builtins/CosmoHound/NS_CosmoHound_Emote_Sequence_Collapse
dependencies {
 usedDataInterfaces [
  &NiagaraDataInterfaceArrayFloat
  &NiagaraDataInterfaceCurve
  &NiagaraDataInterfaceVectorCurve
  &NiagaraDataInterfaceColorCurve
  &NiagaraDataInterfaceVector2DCurve
 ]
 usedDynamicInputs [&FN_FrontEndFloatModifier]
 usedModules [
  &self:SystemSpawnScript.SystemScriptSource.SystemScriptGraph.NiagaraNodeAssignment_0.SetVariables_#1
  &SystemState
  &self:SystemSpawnScript.SystemScriptSource.SystemScriptGraph.NiagaraNodeAssignment_1.SetVariables_#2
 ]
}
emitters {Pop {
  emitterUpdateScript [
   &EmitterState(Life Cycle Mode=::NewEnumerator1, Loop Behavior=::NewEnumerator1, Loop Duration=10, UseLoopDelay=false)
   &SpawnBurst_Instantaneous(Use Spawn Probability=false, UseLoopCountLimit=false)
   &SpawnBurst_Instantaneous(Spawn Group=1, Spawn Time=0.4, Use Spawn Probability=false, UseLoopCountLimit=false) {module=SpawnBurst_Instantaneous003}
  ]
  particleSpawnScript [
   /SetVariables(Particles.VisibilityTag=&SpawnGroup()) {isSetParameters=true}
   &InitializeParticle(
    Color=<rgba 1 0.3325 0 1>
    Lifetime=&SelectFloatFromArray(
     Direct Array Index=&SpawnGroup()
     Float Selection Array=DataInterface() {propertyValues {FloatData [0.5,0.5,0.6,0.3]}}
    )
    Mesh Scale Mode=::NewEnumerator3
    Mesh Uniform Scale=0.8
    Position Offset=&MergeHelper_VectorSwitch(0=<-20 0 -10>, 1=<-20 0 -10>, 2=?, 3=?, Index=&SpawnGroup(), Num=::NewEnumerator0)
    Sprite Size Mode=::NewEnumerator3
    Uniform Sprite Size=40
    UsePositionOffset=false
    Write Material Random=true
   )
   &DynamicMaterialParameters(
    Index 0 Param 1=&RandomRangeFloat(Minimum=0, Maximum=1)
    Index 0 Param 2=&RandomRangeFloat(Minimum=8, Maximum=12)
    Index 0 Param 3=&FloatFromCurve(FloatCurve=DataInterface() {propertyValues {Curve {keys [{interpMode=RCIM_Cubic, tangentMode=RCTM_User, time=0, value=1.6}, {arriveTangent=7.3367, tangentMode=RCTM_User, time=1, value=5}]}}})
    Module.Index0Float0Write=true
    Module.Index0Float1Write=true
    Module.Index0Float2Write=true
    Module.Index0Float3Write=true
   )
   /Game/Effects/Niagara/Modules/Sprite/FN_SpriteUV(End Frame=3, Write to Module Namepsace=true)
   &CameraOffset(Camera Offset Amount=50)
   /SetVariables() {isSetParameters=true}
   /SetVariables(Particles.FlareCameraOffset=5) {isSetParameters=true}
  ]
  particleUpdateScript [
   &ParticleState()
   &ScaleMeshSize(Scale Factor=&Multiply_VectorByFloat(
     Vector=<1 1 1>
     Float=&FloatFromCurve(
      FloatCurve=DataInterface() {propertyValues {Curve {
         defaultValue=1
         keys [
          {interpMode=RCIM_Cubic, tangentMode=RCTM_User, time=0, value=0}
          {
           arriveTangentWeight=0.05
           interpMode=RCIM_Cubic
           leaveTangent=0.0173
           leaveTangentWeight=0.5439
           tangentMode=RCTM_Break
           tangentWeightMode=RCTWM_WeightedBoth
           time=0.15
           value=1
          }
          <at 1 0>
         ]
        }}}
      Scale Curve=$System.FE_Scale
     )
    ))
   &ScaleSpriteSize(
    Uniform Curve Scale=$System.FE_Scale
    Uniform Curve Sprite Scale=DataInterface() {propertyValues {Curve {defaultValue=0, keys [{arriveTangent=3.3458, interpMode=RCIM_Cubic, leaveTangent=3.3458, tangentMode=RCTM_User, time=0, value=1}, <at 1 3>]}}}
   )
   &ScaleColor(
    Scale Alpha=&FloatFromCurve(FloatCurve=DataInterface() {propertyValues {Curve {
        defaultValue=1
        keys [
         {arriveTangent=0.3137, interpMode=RCIM_Cubic, leaveTangent=0.3137, tangentMode=RCTM_User, time=0, value=1}
         {arriveTangent=-2.0958, arriveTangentWeight=0.3333, interpMode=RCIM_Cubic, leaveTangent=-2.0958, tangentMode=RCTM_User, time=1, value=0}
        ]
       }}})
    Scale RGB=&VectorFromCurve(VectorCurve=DataInterface() {propertyValues {
       XCurve {defaultValue=1, keys [{interpMode=RCIM_Cubic, tangentMode=RCTM_User, time=0, value=1}, <at 1 0>]}
       YCurve {defaultValue=1, keys [{interpMode=RCIM_Cubic, tangentMode=RCTM_User, time=0, value=1}, <at 1 0>]}
       ZCurve {defaultValue=1, keys [{interpMode=RCIM_Cubic, tangentMode=RCTM_User, time=0, value=1}, <at 1 0>]}
      }})
    ScaleA=true
    ScaleRGB=true
   )
   &SolveForcesAndVelocity(Clamp Velocity=false, Limit Acceleration=false)
   /SetVariables(
    Particles.AuraColor:LinearColor=&Multiply_LinearColorByFloat(
     Linear Color=$Particles.Initial.Color
     Float=&FloatFromCurve(FloatCurve=DataInterface() {propertyValues {Curve {keys [
          {interpMode=RCIM_Cubic, tangentMode=RCTM_User, time=0, value=0}
          {interpMode=RCIM_Cubic, tangentMode=RCTM_Break, time=0.1, value=2}
          {interpMode=RCIM_Cubic, tangentMode=RCTM_User, time=0.6857, value=0.4}
          <at 0.9903 5>
         ]}}})
    )
    Particles.AuraSize:Vector2f=&Multiply_Vector2D(
     A=$Particles.Initial.SpriteSize
     B=&Vector2DFromCurve(
      Vector2Curve=DataInterface() {propertyValues {
        XCurve {keys [
          {arriveTangent=8.8756, interpMode=RCIM_Cubic, leaveTangent=8.8756, tangentMode=RCTM_User, time=0, value=1}
          {
           arriveTangent=0.6379
           arriveTangentWeight=0.1658
           interpMode=RCIM_Cubic
           leaveTangent=0.0189
           leaveTangentWeight=0.3632
           tangentMode=RCTM_Break
           tangentWeightMode=RCTWM_WeightedBoth
           time=0.3
           value=3.8
          }
          {
           arriveTangent=-0.1745
           arriveTangentWeight=0.7501
           interpMode=RCIM_Cubic
           leaveTangent=-0.1745
           leaveTangentWeight=0.7501
           tangentMode=RCTM_User
           tangentWeightMode=RCTWM_WeightedBoth
           time=1
           value=0
          }
         ]}
        YCurve {keys [
          {arriveTangent=8.8756, interpMode=RCIM_Cubic, leaveTangent=8.8756, tangentMode=RCTM_User, time=0, value=1}
          {
           arriveTangent=0.6379
           arriveTangentWeight=0.1658
           interpMode=RCIM_Cubic
           leaveTangent=0.0189
           leaveTangentWeight=0.3632
           tangentMode=RCTM_Break
           tangentWeightMode=RCTWM_WeightedBoth
           time=0.3
           value=3.8
          }
          {
           arriveTangent=-0.1745
           arriveTangentWeight=0.7501
           interpMode=RCIM_Cubic
           leaveTangent=-0.1745
           leaveTangentWeight=0.7501
           tangentMode=RCTM_User
           tangentWeightMode=RCTWM_WeightedBoth
           time=1
           value=0
          }
         ]}
       }}
      ScaleCurve=&Multiply_Vector2D_ByFloat(Float=$System.FE_Scale, Vector2D=<0.7 0.7>)
     )
    )
    Particles.CollapseRaysScale:Vector3f=&Multiply_VectorByFloat(
     Vector=$Particles.Initial.Scale
     Float=&FloatFromCurve(FloatCurve=DataInterface() {propertyValues {Curve {defaultValue=1, keys [<at 0 0>, {interpMode=RCIM_Cubic, tangentMode=RCTM_User, time=0.1, value=1.1}, <at 1 0>]}}})
    )
    Particles.GlowColor=&Multiply_LinearColorByFloat(
     Linear Color=&MultiplyLinearColor(Linear Color=$Particles.Color, Scale Factor=<1 1 1 1>)
     Float=&FloatFromCurve(
      FloatCurve=DataInterface() {propertyValues {Curve {keys [{interpMode=RCIM_Cubic, tangentMode=RCTM_User, time=0, value=1}, <at 1 0>]}}}
      Scale Curve=0.33
     )
    )
    Particles.GlowSize=&Multiply_Vector2D_ByFloat(
     Float=&FloatFromCurve(
      FloatCurve=DataInterface() {propertyValues {Curve {
         defaultValue=1
         keys [
          {arriveTangent=6.3629, interpMode=RCIM_Cubic, leaveTangent=6.3629, tangentMode=RCTM_User, time=0, value=1.1}
          {arriveTangentWeight=0.0605, interpMode=RCIM_Cubic, tangentMode=RCTM_User, time=0.1814, value=1.8806}
          <at 1 1>
         ]
        }}}
      Scale Curve=1.9999
     )
     Vector2D=$Particles.Initial.SpriteSize
    )
    Particles.RaysColor:LinearColor=&Multiply_LinearColorByFloat(
     Linear Color=&MultiplyLinearColor(
      Linear Color=$Particles.Initial.Color
      Scale Factor=&Vector4FromCurve(Vector4Curve=DataInterface() {propertyValues {
         AlphaCurve {keys [<at 0 1>,<at 1 0>]}
         BlueCurve {keys [{interpMode=RCIM_Cubic, tangentMode=RCTM_User, time=0.6, value=1}, {arriveTangent=18.8498, tangentMode=RCTM_User, time=1, value=5}]}
         GreenCurve {keys [{interpMode=RCIM_Cubic, tangentMode=RCTM_User, time=0.6, value=1}, {arriveTangent=18.8498, tangentMode=RCTM_User, time=1, value=5}]}
         RedCurve {keys [{interpMode=RCIM_Cubic, tangentMode=RCTM_User, time=0.6, value=1}, {arriveTangent=18.8498, tangentMode=RCTM_User, time=1, value=5}]}
        }})
     )
     Float=1
    )
   ) {isSetParameters=true}
   /SetVariables(Particles.Glitch DynP:Vector4f=&MakeVector4(
     X=&RandomRangeFloat(Minimum=0, Maximum=1)
     Y=&Lerp_Float(A=0.5, B=1, Alpha=$Particles.NormalizedAge)
     Z=/Niagara/DynamicInputs/Integer/V2/Round_Float(Float:NiagaraFloat=&FloatFromCurve(
       FloatCurve=DataInterface() {propertyValues {Curve {defaultValue=1, keys [<at 0 1>,<at 0.2035 0>,<at 0.444 0.9782>,<at 0.7 0>,<at 1 0.25>]}}}
       Scale Curve=4
      ))
     W=-1
    )) {isSetParameters=true}
   /SetVariables(
    Particles.CubeCollapseColor:LinearColor=&Multiply_LinearColorByFloat(
     Linear Color=&FNTimeOfDayColor(Night Color=<rgba 0.7517 1 0.0178 1>, Morning Color=<rgba 1 1 1 1>, Day Color=<rgba 1 1 1 1>, Evening Color=<rgba 0.8199 1 0 1>)
     Float=&FloatFromCurve(FloatCurve=DataInterface() {propertyValues {Curve {keys [<at 0 1>,<at 1 1>]}}}, Scale Curve=$System.FE_Scale)
    )
    Particles.CubeCollapseSize:Vector2f=&Multiply_Vector2D_ByFloat(
     Float=&FloatFromCurve(
      FloatCurve=DataInterface() {propertyValues {Curve {keys [
          {arriveTangent=6.3629, interpMode=RCIM_Cubic, leaveTangent=6.3629, tangentMode=RCTM_User, time=0, value=0.1}
          {interpMode=RCIM_Cubic, tangentMode=RCTM_User, time=0.1814, value=0.8806}
          <at 1 0>
         ]}}}
      Scale Curve=$System.FE_Scale
     )
     Vector2D=<80 80>
    )
   ) {isSetParameters=true}
   &Light_Attributes(
    Color=1
    Color Value To Scale:LinearColor=<rgba 5 2.3231 0 1>
    Light Radius:NiagaraFloat=&FloatFromCurve(
     FloatCurve=DataInterface() {propertyValues {Curve {keys [{interpMode=RCIM_Cubic, tangentMode=RCTM_User, time=0, value=1}, <at 1 0>]}}}
     Scale Curve=170
    )
    Radius=1
    ScaleA=true
    ScaleRGB=true
   )
   /SetVariables(
    Particles.FlareGlowSize:Vector2f=&Multiply_Vector2D_ByFloat(
     Float=&FloatFromCurve(FloatCurve=DataInterface() {propertyValues {Curve {keys [
          {interpMode=RCIM_Cubic, tangentMode=RCTM_User, time=0, value=0}
          {interpMode=RCIM_Cubic, tangentMode=RCTM_User, time=0.1, value=1}
          {interpMode=RCIM_Cubic, tangentMode=RCTM_User, time=1, value=0.5}
         ]}}})
     Vector2D=$Particles.Initial.SpriteSize
    )
    Particles.FlareRaysScale:Vector3f=&Multiply_VectorByFloat(
     Vector=$Particles.Initial.Scale
     Float=&FloatFromCurve(
      FloatCurve=DataInterface() {propertyValues {Curve {defaultValue=1, keys [<at 0 0>, {interpMode=RCIM_Cubic, tangentMode=RCTM_User, time=0.1, value=1}, <at 1 0>]}}}
      Scale Curve=0.15
     )
    )
    Particles.FlareSize:Vector2f=&Multiply_Vector2D_ByFloat(
     Float=&FloatFromCurve(
      FloatCurve=DataInterface() {propertyValues {Curve {keys [
          {arriveTangent=0.1416, interpMode=RCIM_Cubic, leaveTangent=0.1416, tangentMode=RCTM_User, time=0, value=0}
          {arriveTangent=0.1416, interpMode=RCIM_Cubic, leaveTangent=0.1416, tangentMode=RCTM_User, time=0.1, value=1}
          <at 1 0>
         ]}}}
      Scale Curve=0.7
     )
     Vector2D=$Particles.Initial.SpriteSize
    )
   ) {isSetParameters=true}
  ]
  rendererClasses [
   &NiagaraSpriteRendererProperties
   &NiagaraMeshRendererProperties
   &NiagaraLightRendererProperties
  ]
  renderers [
   {rendererClass=&NiagaraSpriteRendererProperties}
   {rendererClass=&NiagaraMeshRendererProperties, rendererIndex=1}
   {rendererClass=&NiagaraSpriteRendererProperties, rendererIndex=2}
   {rendererClass=&NiagaraSpriteRendererProperties, rendererIndex=3}
   {rendererClass=&NiagaraLightRendererProperties, rendererIndex=4}
   {rendererClass=&NiagaraSpriteRendererProperties, rendererIndex=5}
   {rendererClass=&NiagaraSpriteRendererProperties, rendererIndex=6}
   {rendererClass=&NiagaraMeshRendererProperties, rendererIndex=7}
  ]
 }}
summary {
 systemName=NS_CosmoHound_Emote_Sequence_Collapse
 emitters [{
   emitterName=Pop
   rendererClasses [
    &NiagaraSpriteRendererProperties
    &NiagaraMeshRendererProperties
    &NiagaraLightRendererProperties
   ]
  }]
 userVariables [
  User.bIsFrontEnd {type=NiagaraBool}
  User.bIsFrontEndPreview {type=NiagaraBool}
  User.Variant {type=NiagaraFloat, defaultValue=0, description=""}
 ]
}
