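# Niagara idle glow for the CosmoHound cosmetic companion, attaching flames, lightning and glow to its skeletal mesh.
# CubicFlame uses particle partitions to spread rising sprite and mesh flames across bones, Hunt_Lightning flashes short-lived meshes, and NeckFlameMesh holds bone-following meshes, glow sprites and a flickering light.
# User.VariantID switches bones and rates, with variant 1 removing lightning and neck mesh/sprites.
# bisHeadHidden blanks flames and neck mesh; front-end modifiers adjust scale and brightness.
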NiagaraSystem /CosmeticCompanions/FX/Builtins/CosmoHound/NS_CosmoHound_Idle_Glow
dependencies {
 usedDataInterfaces [
  &NiagaraDataInterfaceCurve
  &NiagaraDataInterfaceArrayInt32
  &NiagaraDataInterfaceArrayFloat3
  /Script/Niagara.NiagaraDataInterfaceArrayQuat
  &NiagaraDataInterfaceMeshRendererInfo
  &NiagaraDataInterfaceVector2DCurve
  &NiagaraDataInterfaceVectorCurve
  &NiagaraDataInterfaceSkeletalMesh
  &NiagaraDataInterfaceColorCurve
 ]
 usedDynamicInputs []
 usedModules [
  &self:SystemSpawnScript.SystemScriptSource.SystemScriptGraph.NiagaraNodeAssignment_0.SetVariables_#1
  &self:SystemSpawnScript.SystemScriptSource.SystemScriptGraph.NiagaraNodeAssignment_1.SetVariables_#2
  &SystemState
  &self:SystemSpawnScript.SystemScriptSource.SystemScriptGraph.NiagaraNodeAssignment_2.SetVariables_#3
  &self:SystemSpawnScript.SystemScriptSource.SystemScriptGraph.NiagaraNodeAssignment_3.SetVariables_#4
 ]
}
emitters {
 CubicFlame {
  emitterUpdateScript [
   &EmitterState()
   &SpawnRate(
    SpawnRate=&FloatFromCurve(
     FloatCurve=DataInterface() {propertyValues {Curve {keys [
         {interpMode=RCIM_Cubic, tangentMode=RCTM_User, time=0, value=0.25}
         {interpMode=RCIM_Cubic, tangentMode=RCTM_User, time=0.05, value=1}
         {interpMode=RCIM_Cubic, tangentMode=RCTM_User, time=0.75, value=1}
         {arriveTangent=-5.7829, tangentMode=RCTM_User, time=1, value=0.25}
        ]}}}
     CurveIndex=$System.Timeline.NormalizedTime
     Scale Curve=&MakeFloatFromInt(Int=&MergeHelper_IntSwitch(0=100, 1=30, 2=?, 3=?, 4=?, 5=?, Index=&MakeIntFromFloat(Float=$User.VariantID), Num=::NewEnumerator0))
    )
    Use Spawn Probability=true
   )
   &SpawnRate(
    SpawnGroup=1
    SpawnRate=&MakeFloatFromInt(Int=&MergeHelper_IntSwitch(0=9, 1=3, 2=?, 3=?, 4=?, 5=?, Index=&MakeIntFromFloat(Float=$User.VariantID), Num=::NewEnumerator0))
    Use Spawn Probability=false
   ) {module=SpawnRate001}
  ]
  particleSpawnScript [
   &PartitionParticles(
    Number Of Partitions=&MergeHelper_IntSwitch(0=9, 1=4, 2=?, 3=?, 4=?, 5=?, Index=&MakeIntFromFloat(Float=$User.VariantID), Num=::NewEnumerator0)
    Partition Mode:ENiagara_PartitionMode_v2=::NewEnumerator5
   )
   &InitializeParticle(
    Lifetime=&RandomRangeFloat(Minimum=0.5, Maximum=1.5)
    Mesh Scale Mode=::NewEnumerator3
    Mesh Uniform Scale=&RandomRangeFloat(Minimum=0.15, Maximum=0.4)
    Sprite Size Max=<15 20>
    Sprite Size Min=<10 15>
    Sprite Size Mode=::NewEnumerator2
    UsePositionOffset=false
    Write Material Random=true
   )
   &SkeletalMeshLocation(
    Bone / Socket Index=&MergeHelper_IntSwitch(
     0=&SelectIntFromArray(
      Direct Array Index=$Particles.PartitionParticles.Partition
      Int Selection Array=DataInterface() {propertyValues {IntData [2,2,2,3,4,2,6,7,8,9]}}
     )
     1=&SelectIntFromArray(Direct Array Index=$Particles.PartitionParticles.Partition, Int Selection Array=DataInterface() {propertyValues {IntData [4,5,6,7]}})
     2=?
     3=?
     4=?
     5=?
     Index=&MakeIntFromFloat(Float=$User.VariantID)
     Num=::NewEnumerator0
    )
    Bone Sampling Mode=::NewEnumerator4
    MaskSpawnGroup=false
    Orientation Sampling=::NewEnumerator1
    Sampled Position Offset=&MergeHelper_VectorSwitch(
     0=&SelectVectorFromArray(
      Direct Array Index=$Particles.PartitionParticles.Partition
      Vector Selection Array=DataInterface() {propertyValues {FloatData [<0 0 0>, <0 0 0>, <0 0 0>, <3 -3 0>, <-3 3 0>, <0 0 0>, <0 0 0>, <0 0 0>, <0 0 0>, <0 0 0>]}}
     )
     1=<0 0 0>
     2=?
     3=?
     Index=&MakeIntFromFloat(Float=$User.VariantID)
     Num=::NewEnumerator0
    )
    Skeletal Mesh=$System.Skeletal Mesh
   )
   &AddVelocity(
    Cone Axis=&MaskVector(
     Mask Source=1
     Mask Parameter=&MakeIntFromFloat(Float=$User.VariantID)
     Mask=0
     Masked Vector=&RotateVectorByQuaternion(
      Vector To Rotate=&ReturnMeshOrientationAxis(Mesh Orientation Axis=::XAxis, Mesh Orientation Quat=$Particles.SkeletalMeshLocation.SampledRotation)
      Quaternion=/Niagara/DynamicInputs/Arrays/SelectQuaternionFromArray(
       Array Sampling Mode:ENiagaraArraySamplingMode=::NewEnumerator1
       Direct Array Index:NiagaraInt32=$Particles.PartitionParticles.Partition
       Interpolation:NiagaraFloat=?
       Direct Array Mode:ENiagara_WrapClamp=::NewEnumerator0
       Evaluation Type:ENiagaraRandomnessEvaluation=::NewEnumerator0
       Normalized Interpolation:NiagaraBool=true
       Recalculate Random Each Loop:NiagaraBool=false
       Randomness Mode:ENiagaraRandomnessMode=::NewEnumerator0
       Random Seed:NiagaraInt32=?
       Fixed Random Seed:NiagaraBool=false
       Quaternion Selection Array:NiagaraDataInterfaceArrayQuat=DataInterface() {propertyValues {
         GpuSyncMode=SyncCpuToGpu
         MaxElements=0
         QuatData [
          <0 0 0 1>
          <0 0 0 1>
          <-0.0872 0 0 0.9962>
          <-0.0872 0 0 0.9962>
          <0 0 0 1>
          <-0.5736 0 0 0.8192>
          <-0.4226 0 0 0.9063>
          <-0.342 0 0 0.9397>
          <-0.2588 0 0 0.9659>
          <-0.1736 0 0 0.9848>
         ]
        }}
       Enable Spherical Linear Interpolation:NiagaraBool=?
      )
     )
     Passthrough Vector=&RotateVectorByQuaternion(
      Vector To Rotate=&ReturnMeshOrientationAxis(Mesh Orientation Axis=::XAxis, Mesh Orientation Quat=$Particles.SkeletalMeshLocation.SampledRotation)
      Quaternion=&SelectQuaternionFromArray(
       Array Sampling Mode:ENiagaraArraySamplingMode=::NewEnumerator1
       Direct Array Index:NiagaraInt32=$Particles.PartitionParticles.Partition
       Interpolation:NiagaraFloat=?
       Direct Array Mode:ENiagara_WrapClamp=::NewEnumerator0
       Evaluation Type:ENiagaraRandomnessEvaluation=::NewEnumerator0
       Normalized Interpolation:NiagaraBool=true
       Recalculate Random Each Loop:NiagaraBool=false
       Randomness Mode:ENiagaraRandomnessMode=::NewEnumerator0
       Random Seed:NiagaraInt32=?
       Fixed Random Seed:NiagaraBool=false
       Quaternion Selection Array:NiagaraDataInterfaceArrayQuat=DataInterface() {propertyValues {GpuSyncMode=SyncCpuToGpu, MaxElements=0, QuatData [<-0.4226 0 0 0.9063>, <-0.342 0 0 0.9397>, <-0.2588 0 0 0.9659>, <-0.1736 0 0 0.9848>]}}
       Enable Spherical Linear Interpolation:NiagaraBool=?
      )
     )
    )
    Module.UseConeFalloff=false
    Module.UseVelDistribution=false
    Override Seed=false
    Velocity Mode=::NewEnumerator2
    Velocity Speed=&MergeHelper_FloatSwitch(
     0=&MaskFloat(Mask Source=1, Mask Parameter=$Particles.PartitionParticles.Partition, Mask=3, Masked Float=-5, Passthrough Float=5)
     1=5
     2=?
     3=?
     Index=&MakeIntFromFloat(Float=$User.VariantID)
     Num=::NewEnumerator0
    )
   ) {enabled=false}
   &AddVelocity(
    Cone Axis=<0 0 1>
    Module.UseConeFalloff=false
    Module.UseVelDistribution=false
    Override Seed=false
    Rotation Coordinate Space=::World
    Velocity Mode=::NewEnumerator2
    Velocity Speed=&MergeHelper_FloatSwitch(
     0=&MaskFloat(Mask Source=1, Mask Parameter=$Particles.PartitionParticles.Partition, Mask=3, Masked Float=5, Passthrough Float=5)
     1=5
     2=?
     3=?
     Index=&MakeIntFromFloat(Float=$User.VariantID)
     Num=::NewEnumerator0
    )
   ) {module=AddVelocity001}
   &ShapeLocation(
    Additional Quaternion Rotation=true
    Additional Yaw / Pitch / Roll=<0 0.25 0>
    Disc Coverage=1
    MaskSpawnGroup=false
    Offset Coordinate Space=::Local
    Offset Mode=::NewEnumerator0
    Override Seed=false
    Ring Radius=&MergeHelper_FloatSwitch(
     0=&MakeCustomFloatFromBool(
      Bool=&SetBoolByIntComparison(A=$Particles.PartitionParticles.Partition, B=2, Comparison Type=::NewEnumerator7)
      True Float=8
      False Float=0
     )
     1=0
     2=?
     3=?
     Index=&MakeIntFromFloat(Float=$User.VariantID)
     Num=::NewEnumerator0
    )
    Rotation Mode=::NewEnumerator3
    Rotation Quaternion=$Particles.SkeletalMeshLocation.SampledRotation
    Shape Origin=$Particles.SkeletalMeshLocation.SampledPosition
    Shape Primitive=::NewEnumerator4
    UseOwnerScale=true
   )
   &ShapeLocation(
    Additional Quaternion Rotation=false
    Box / Plane Mode:ENiagara_BoxPlaneMode=::NewEnumerator1
    MaskSpawnGroup=false
    Override Seed=false
    Plane Size:Vector2f=&MergeHelper_Vector2Switch(
     0=&MakeCustomVector2DFromBool(
      Bool=&SetBoolByIntComparison(A=$Particles.PartitionParticles.Partition, B=3, Comparison Type=::NewEnumerator0)
      True Vector2D=<7 0>
      False Vector2D=<0 0>
     )
     1=<7 0>
     2=?
     3=?
     4=?
     Index=&MakeIntFromFloat(Float=$User.VariantID)
     Num=::NewEnumerator0
    )
    Rotation Mode=::NewEnumerator3
    Rotation Quaternion=$Particles.SkeletalMeshLocation.SampledRotation
    Shape Primitive=::NewEnumerator2
    UseOwnerScale=true
   ) {module=ShapeLocation001}
   /SetVariables(Particles.VisibilityTag=&SpawnGroup(Max Spawn Group=2)) {isSetParameters=true}
   &DynamicMaterialParameters(
    Index 0 Param 1=&RandomRangeFloat(Minimum=0, Maximum=1)
    Index 0 Param 2=&RandomRangeFloat(Minimum=0, Maximum=1)
    Module.Index0Float0Write=true
    Module.Index0Float1Write=true
    Module.Index0Float2Write=true
    Module.Index0Float3Write=true
   )
   &ApplyInitialForces()
   &SubUVAnimation(
    Mesh Renderer=DataInterface() {propertyValues {MeshRenderer=&self:CubicFlame_0.NiagaraMeshRendererProperties_0}}
    Output Alternate Renderer Binding=true
    Renderer Type:ENiagaraMeshOrSprite=::NewEnumerator1
    UseEndFrame=false
    UseStartFrame=false
   )
   /SetVariables() {isSetParameters=true}
  ]
  particleUpdateScript [
   &ParticleState()
   &ScaleMeshSize(Scale Factor=&Multiply_VectorByFloat(
     Vector=<1 1 1>
     Float=&FloatFromCurve(FloatCurve=DataInterface() {propertyValues {Curve {keys [<at 0 0 RCIM_Cubic>, {interpMode=RCIM_Cubic, tangentMode=RCTM_User, time=0.3, value=1}, <at 1 0>]}}})
    ))
   &ScaleSpriteSize(
    Non-Uniform Curve Scale=&RandomRangeVector2D(
     Minimum=<0.8 0.8>
     Maximum=&Multiply_Vector2D_ByFloat(
      Float=&FloatFromCurve(
       FloatCurve=DataInterface() {propertyValues {Curve {keys [<at 0.55 1 RCIM_Cubic>,<at 1 0.7>]}}}
       CurveIndex=$Particles.PartitionParticles.NormalizedPartition
      )
      Vector2D=<1.3 1.3>
     )
    )
    Non-Uniform Curve Sprite Scale=DataInterface() {propertyValues {
      XCurve {keys [
        {arriveTangent=3.068, interpMode=RCIM_Cubic, leaveTangent=3.068, tangentMode=RCTM_User, time=0, value=0.5}
        {interpMode=RCIM_Cubic, tangentMode=RCTM_User, time=0.2, value=1}
        {arriveTangent=-2.4882, tangentMode=RCTM_User, time=1, value=0}
       ]}
      YCurve {keys [
        {arriveTangent=3.3898, interpMode=RCIM_Cubic, leaveTangent=3.3898, tangentMode=RCTM_User, time=0, value=0.1}
        {arriveTangent=0.0355, interpMode=RCIM_Cubic, leaveTangent=0.0355, tangentMode=RCTM_User, time=1, value=2}
       ]}
     }}
    Scale Sprite Size Mode=::NewEnumerator3
   )
   &CameraOffset(Camera Offset Amount=-0.5)
   &ScaleColor(
    Scale Alpha=&FloatFromCurve(
     FloatCurve=DataInterface() {propertyValues {Curve {keys [<at 0 0>,<at 0.1 1>,<at 1 0>]}}}
     Scale Curve=&MakeCustomFloatFromBool(Bool=$User.bisHeadHidden, True Float=0, False Float=1)
    )
    Scale RGB=&Multiply_VectorByFloat(
     Vector=<1 1 1>
     Float=&FloatFromCurve(
      FloatCurve=DataInterface() {propertyValues {Curve {keys [<at 0 1>,<at 0.084 3>,<at 1 0>]}}}
      Scale Curve=&MakeCustomFloatFromBool(Bool=$User.bisHeadHidden, True Float=0, False Float=1)
     )
    )
    ScaleA=true
    ScaleRGB=true
   )
   &SkeletalMeshLocation(
    Bone / Socket Index=&MergeHelper_IntSwitch(
     0=&SelectIntFromArray(
      Direct Array Index=$Particles.PartitionParticles.Partition
      Int Selection Array=DataInterface() {propertyValues {IntData [2,2,2,3,4,2,6,7,8,9]}}
     )
     1=&SelectIntFromArray(Direct Array Index=$Particles.PartitionParticles.Partition, Int Selection Array=DataInterface() {propertyValues {IntData [4,5,6,7]}})
     2=?
     3=?
     4=?
     5=?
     Index=&MakeIntFromFloat(Float=$User.VariantID)
     Num=::NewEnumerator0
    )
    Bone Sampling Mode=::NewEnumerator4
    Initialize Particle Position To Sampled Position=false
    MaskSpawnGroup=false
    Position Sampling=::NewEnumerator1
    Skeletal Mesh=$System.Skeletal Mesh
    Use Sampled Mesh Attribute as Ribbon ID=false
   ) {module=SkeletalMeshLocation001}
   &AccelerationForce(Acceleration=<0 0 1>)
   &ScaleVelocity(Velocity Scale=&MaskVector(
     Mask Source=0
     Mask Parameter=?
     Mask=0
     Masked Vector=&VectorFromFloat(Value=&FloatFromCurve(FloatCurve=DataInterface() {propertyValues {Curve {keys [{arriveTangent=-2.9609, interpMode=RCIM_Cubic, leaveTangent=-2.9609, tangentMode=RCTM_User, time=0, value=1}, <at 1 0>]}}}))
     Passthrough Vector=&VectorFromFloat(Value=&FloatFromCurve(
       FloatCurve=DataInterface() {propertyValues {Curve {keys [{arriveTangent=-0.0171, interpMode=RCIM_Cubic, leaveTangent=-0.0171, tangentMode=RCTM_User, time=0, value=1}, <at 1 0.1>]}}}
       Scale Curve=3
      ))
    ))
   &SolveForcesAndVelocity(Clamp Velocity=false, Limit Acceleration=false)
   /SetVariables(Particles.CubeDyn00:Vector4f=&MakeVector4(
     X=&FloatFromCurve(FloatCurve=DataInterface() {propertyValues {Curve {keys [<at 0 0>,<at 1 4>]}}})
     Y=1
     Z=&Waveform(
      Global Input Position=&RandomRangeFloat(Minimum=0, Maximum=1)
      "Waveform [1]"=::NewEnumerator0
      "[1] Frequency"=1.3
      "[1] Amplitude Min/Max"=<0.2 0.6>
      "[1] Scale"=2
     )
     W=0
    )) {isSetParameters=true}
  ]
  rendererClasses [&NiagaraSpriteRendererProperties, &NiagaraMeshRendererProperties]
  renderers [
   {rendererClass=&NiagaraSpriteRendererProperties}
   {rendererClass=&NiagaraMeshRendererProperties, rendererIndex=1}
  ]
 }
 Hunt_Lightning {
  emitterUpdateScript [
   &EmitterState()
   &SpawnRate(
    Spawn Probability=&MergeHelper_FloatSwitch(
     0=/Niagara/DynamicInputs/Helpers/DistributionCurveFloat(
      Minimum:NiagaraFloat=0.25
      Maximum:NiagaraFloat=1
      Distribution Curve:NiagaraDataInterfaceCurve=DataInterface() {propertyValues {Curve {
         defaultValue=0
         keys [{interpMode=RCIM_Cubic, tangentMode=RCTM_User, time=0.5, value=0}, {arriveTangent=-0.0819, tangentMode=RCTM_User, time=1, value=1}]
        }}}
      Randomness Mode:ENiagaraRandomnessMode=::NewEnumerator0
      Evaluation Type:ENiagaraRandomnessEvaluation=::NewEnumerator1
      Seed:NiagaraInt32=0
      Fixed Random Seed:NiagaraBool=false
     )
     1=0
     2=?
     3=?
     Index=&MakeIntFromFloat(Float=$User.VariantID)
     Num=::NewEnumerator0
    )
    SpawnRate=9
    Use Spawn Probability=true
   )
  ]
  particleSpawnScript [
   &PartitionParticles() {enabled=false}
   &InitializeParticle(
    Color=<rgba 2 0.1311 0 1>
    Lifetime Max=0.5
    Lifetime Min=0.2
    Lifetime Mode=::NewEnumerator1
    Mesh Scale=&Multiply_VectorByFloat(
     Vector=&Lerp_Vector(
      A=&RandomRangeVector(Minimum=<0.5 0.4 0.4>, Maximum=<0.75 0.6 0.6>)
      B=&RandomRangeVector(Minimum=<1 1 1>, Maximum=<1.5 2 2>)
      Alpha=&FloatFromCurve(
       FloatCurve=DataInterface() {propertyValues {Curve {keys [{interpMode=RCIM_Cubic, tangentMode=RCTM_User, time=0.5, value=0}, <at 1 1>]}}}
       CurveIndex=&RandomRangeFloat(Minimum=0, Maximum=1)
      )
     )
     Float=0.2
    )
    Mesh Scale Mode=::NewEnumerator4
    UsePositionOffset=false
    Write Material Random=true
   )
   &CameraOffset(Camera Offset Amount=20) {enabled=false}
   &SkeletalMeshLocation(
    Bone Sampling Mode=::NewEnumerator4
    MaskSpawnGroup=false
    Sampled Position Offset=<0 -2.5 0>
    Sampled Position Offset Coordinate Space:ENiagaraSkelMeshTransforms=::NewEnumerator2
    Skeletal Mesh=$System.Skeletal Mesh
   )
   &ShapeLocation(Box Size=<0.5 0.5 0.5>, MaskSpawnGroup=false, Override Seed=false, Shape Primitive=::NewEnumerator2, Surface Only=false, UseOwnerScale=true)
   &InitialMeshOrientation(Rotation=&RandomRangeVector(Minimum=<-0.5 -0.5 -0.5>, Maximum=<0.5 0.5 0.5>), Use Rotation Vector=true) {module=InitialMeshOrientation002}
   /SetVariables(Particles.Random=&RandomRangeFloat(Minimum=0, Maximum=1)) {isSetParameters=true}
  ]
  particleUpdateScript [
   &ParticleState()
   &ScaleMeshSize()
   &ScaleColor(
    Scale Alpha=&Multiply_Float(
     A=10
     B=&FloatFromCurve(FloatCurve=DataInterface() {propertyValues {Curve {keys [
          {interpMode=RCIM_Cubic, tangentMode=RCTM_User, time=0, value=0}
          {interpMode=RCIM_Cubic, tangentMode=RCTM_User, time=0.1, value=0}
          {interpMode=RCIM_Cubic, tangentMode=RCTM_User, time=0.2, value=0.8}
          {interpMode=RCIM_Cubic, tangentMode=RCTM_User, time=1, value=0}
         ]}}})
    )
    Scale RGB=&Multiply_VectorByFloat(
     Vector=&Multiply_VectorByFloat(Vector=<1 1 1>, Float=&FN_FrontEndFloatModifier(Gameplay=1, Front End=0.5))
     Float=&Multiply_Float(
      A=&RandomRangeFloat(Minimum=0.5, Maximum=1, Evaluation Type=::NewEnumerator1)
      B=&Multiply_Float(A=&FNTimeOfDayScalar(Night Scale=0.2, Morning Scale=1, Evening Scale=1), B=10)
     )
    )
    ScaleA=true
    ScaleRGB=true
   )
   &DynamicMaterialParameters(
    Index 0 Param 1=&FloatFromCurve(FloatCurve=DataInterface() {propertyValues {Curve {keys [{tangentMode=RCTM_User, time=0, value=1}, {tangentMode=RCTM_User, time=1, value=1}]}}})
    Index 0 Param 2=0.7
    Index 0 Param 3=&FloatFromCurve(FloatCurve=DataInterface() {propertyValues {Curve {keys [
         {arriveTangent=4.0072, interpMode=RCIM_Cubic, leaveTangent=4.0072, tangentMode=RCTM_User, time=0, value=0}
         {arriveTangent=2.7083, interpMode=RCIM_Cubic, leaveTangent=2.7083, time=0.2543, value=0.6225}
         {arriveTangent=1.8473, interpMode=RCIM_Cubic, leaveTangent=1.8473, time=0.6, value=1.625}
         {arriveTangent=1.6993, interpMode=RCIM_Cubic, leaveTangent=1.6993, tangentMode=RCTM_User, time=1, value=2}
        ]}}})
    Index 0 Param 4=5
    Index 1 Param 1=0.5
    Index 1 Param 2=$Particles.Random
    Index 1 Param 3=1.3
    Index 1 Param 4=1
    Index 2 Param 1=&FloatFromCurve(FloatCurve=DataInterface() {propertyValues {Curve {keys [<at 0 1>,<at 1 0.3>]}}})
    Module.Index0Float0Write=true
    Module.Index0Float1Write=true
    Module.Index0Float2Write=true
    Module.Index0Float3Write=true
    Module.Index1Float0Write=true
    Module.Index1Float1Write=true
    Module.Index1Float2Write=true
    Module.Index1Float3Write=true
    Module.Index2Float0Write=true
    Module.Index2Float1Write=true
    Module.Index2Float2Write=true
    Module.Index2Float3Write=true
    Write Parameter Index 1=::NewEnumerator1
    Write Parameter Index 2=::NewEnumerator1
   )
   &SolveForcesAndVelocity(Clamp Velocity=false, Limit Acceleration=false)
   /SetVariables(Particles.Lightning glow color:LinearColor=&MakeLinearColorFromVectorAndFloat(
     "Vector (RGB)"=&MakeVectorFromLinearColorRGB(Color=&MultiplyLinearColor(
       Linear Color=<rgba 4 1.6071 0.1968 1>
       Scale Factor=&MakeVector4_FromFloat(Float=&RandomRangeFloat(Minimum=0.5, Maximum=1, Evaluation Type=::NewEnumerator1))
      ))
     "Float (Alpha)"=&Multiply_Float(
      A=&FloatFromCurve(FloatCurve=DataInterface() {propertyValues {Curve {keys [<at 0 1 RCIM_Cubic>, {interpMode=RCIM_Cubic, leaveTangent=-3.6307, time=0.9, value=0}]}}})
      B=1
     )
    )) {isSetParameters=true}
   /SetVariables(Particles.Glow Dy02:Vector4f=&MakeVector4(
     X=&FloatFromCurve(FloatCurve=DataInterface() {propertyValues {Curve {keys [{tangentMode=RCTM_User, time=0, value=0.2}, {tangentMode=RCTM_User, time=0.5, value=1}]}}})
     Y=0.7
     Z=0
     W=5
    )) {isSetParameters=true}
   /SetVariables(Particles.Glow Dy03:Vector4f=&MakeVector4(
     X=&FloatFromCurve(
      FloatCurve=DataInterface() {propertyValues {Curve {keys [{interpMode=RCIM_Cubic, tangentMode=RCTM_User, time=0, value=1}, <at 1 0.5>]}}}
      Scale Curve=0.7
     )
     Y=$Particles.Random
     Z=1.3
     W=1
    )) {isSetParameters=true}
  ]
  rendererClasses [&NiagaraMeshRendererProperties]
  renderers [
   {rendererClass=&NiagaraMeshRendererProperties, rendererIndex=0}
   {rendererClass=&NiagaraMeshRendererProperties, rendererIndex=1}
  ]
 }
 NeckFlameMesh {
  emitterUpdateScript [
   &EmitterState(
    Inactive Response=::NewEnumerator1
    Life Cycle Mode=::NewEnumerator1
    Loop Behavior=::NewEnumerator1
    Loop Duration Mode=::NewEnumerator1
    UseLoopDelay=false
   )
   &SpawnBurst_Instantaneous(Use Spawn Probability=false, UseLoopCountLimit=false)
   &SpawnBurst_Instantaneous(Spawn Group=1, Use Spawn Probability=false, UseLoopCountLimit=false) {module=SpawnBurst_Instantaneous001}
  ]
  particleSpawnScript [
   &PartitionParticles()
   &InitializeParticle(
    Color=<rgba 10 10 10 1>
    Mesh Scale=<1.3 0.86 0.88>
    Mesh Scale Mode=::NewEnumerator4
    Sprite Size Mode=::NewEnumerator3
    Uniform Sprite Size=65
    UsePositionOffset=false
    Write Material Random=true
   )
   &InitialMeshOrientation(Mesh Orientation Mode=::NewEnumerator2, Rotation=<0 0 0.015>, Use Rotation Vector=true)
   /SetVariables(Particles.GlowColor=&Multiply_LinearColorByFloat(Linear Color=<rgba 2 0.5402 0.0681 1>, Float=$System.FE_Scale)) {isSetParameters=true}
   /SetVariables(Particles.SideSpriteRotation=75, Particles.SideSpriteSize:Vector2f=<60 90>) {isSetParameters=true}
   &DynamicMaterialParameters(
    Index 0 Param 1=3
    Index 0 Param 2=1
    Index 0 Param 3=1
    Index 0 Param 4=1
    Module.Index0Float0Write=false
    Module.Index0Float1Write=false
    Module.Index0Float2Write=true
    Module.Index0Float3Write=false
   )
   &CameraOffset()
   /SetVariables(Particles.NeckGlowSize:Vector2f=&Multiply_Vector2D_ByFloat(Float=0.5, Vector2D=$Particles.SpriteSize)) {isSetParameters=true}
  ]
  particleUpdateScript [
   &ParticleState(Kill Particles When Lifetime Has Elapsed=false, Let Infinitely Lived Particles Die When Emitter Deactivates=true)
   &ScaleSpriteSize(
    Scale Sprite Size Mode=::NewEnumerator0
    Uniform Scale Factor=&MergeHelper_FloatSwitch(0=1, 1=0, 2=?, 3=?, Index=&MakeIntFromFloat(Float=$User.VariantID), Num=::NewEnumerator0)
   )
   /SetVariables(Particles.VisibilityTag=$Particles.PartitionParticles.Partition) {isSetParameters=true}
   &SkeletalMeshLocation(
    Bone / Socket Index=&SelectIntFromArray(Direct Array Index=$Particles.PartitionParticles.Partition, Int Selection Array=DataInterface() {propertyValues {IntData [0,2]}})
    Bone Sampling Mode=::NewEnumerator4
    MaskSpawnGroup=false
    Sampled Position Offset=/Niagara/DynamicInputs/Helpers/MaskPosition(
     Mask Source:NiagaraInt32=1
     Mask Parameter:NiagaraInt32=$Particles.PartitionParticles.Partition
     Mask:NiagaraInt32=0
     Masked Position:NiagaraPosition=<0 0 1>
     Passthrough Position:NiagaraPosition=&FN_FrontEndVector3Modifier(Gameplay=<0.25 0 0>, Front End=<0.5 0 0>)
    )
    Skeletal Mesh=$System.Skeletal Mesh
   )
   &ScaleMeshSize(Scale Factor=&MakeCustomVectorFromBool(
     Bool=$User.bisHeadHidden
     True Vector=<0 0 0>
     False Vector=&Multiply_Vector(
      A=&FN_FrontEndVector3Modifier(Gameplay=<1 1 1>, Front End=<0.5 1 1>)
      B=&VectorFromCurve(
       VectorCurve=DataInterface() {propertyValues {
         XCurve {keys [
           {arriveTangent=1.9574, interpMode=RCIM_Cubic, leaveTangent=1.9574, tangentMode=RCTM_User, time=0, value=1}
           {interpMode=RCIM_Cubic, tangentMode=RCTM_User, time=0.75, value=1.5}
           {interpMode=RCIM_Cubic, tangentMode=RCTM_User, time=1, value=1}
          ]}
         YCurve {keys [<at 0 1>,<at 1 1>]}
         ZCurve {keys [<at 0 1>,<at 1 1>]}
        }}
       CurveIndex=$System.Timeline.NormalizedTime
       Scale Curve=&Multiply_VectorByFloat(
        Vector=<1 1 1>
        Float=&MergeHelper_FloatSwitch(0=1, 1=0, 2=?, 3=?, Index=&MakeIntFromFloat(Float=$User.VariantID), Num=::NewEnumerator0)
       )
      )
     )
    ))
   &Light_Attributes(
    Color=1
    Color Value To Scale:LinearColor=$Particles.GlowColor
    Diffuse Scale=1
    Enabled=1
    Falloff Exponent=1
    Light Exponent=5
    Light Position Offset:Vector3f=<0 0 4>
    Light Radius=15
    Light Specular Scale=0
    Position=1
    Position Offset Coordinate Space:ENiagaraCoordinateSpace=::Local
    Radius=1
    Scale RGB:Vector3f=&Multiply_VectorByFloat(
     Vector=<5 5 5>
     Float=&Waveform("Waveform [1]"=::NewEnumerator3, "[1] Frequency"=1.37, "[1] Amplitude Min/Max"=<20 40>, "[1] Bias"=-0.1216)
    )
    ScaleA=true
    ScaleRGB=true
    Specular Scale=1
    UsePositionOffset=true
   )
   /SetVariables(
    Particles.SideSpriteAlignment:Vector3f=&ReturnMeshOrientationAxis(Mesh Orientation Axis=::ZAxis, Mesh Orientation Quat=$Particles.SkeletalMeshLocation.SampledRotation)
    Particles.SideSpriteFacing:Vector3f=&ReturnMeshOrientationAxis(Mesh Orientation Axis=::XAxis, Mesh Orientation Quat=$Particles.SkeletalMeshLocation.SampledRotation)
    Particles.SideSpriteSize:Vector2f=&Multiply_Vector2D_ByFloat(
     Float=&MergeHelper_FloatSwitch(0=1, 1=0, 2=?, 3=?, Index=&MakeIntFromFloat(Float=$User.VariantID), Num=::NewEnumerator0)
     Vector2D=$Particles.Initial.SideSpriteSize
    )
   ) {isSetParameters=true}
   /SetVariables(
    Particles.NeckGlowAlignment:Vector3f=&ReturnMeshOrientationAxis(Mesh Orientation Axis=::ZAxis, Mesh Orientation Quat=$Particles.SkeletalMeshLocation.SampledRotation)
    Particles.NeckGlowFacing:Vector3f=&ReturnMeshOrientationAxis(Mesh Orientation Axis=::XAxis, Mesh Orientation Quat=$Particles.SkeletalMeshLocation.SampledRotation)
    Particles.NeckGlowPosition:NiagaraPosition=&AddVectorToPosition(Position=$Particles.SkeletalMeshLocation.SampledPosition, Vector=<0 0 3>)
    Particles.NeckGlowSize:Vector2f=&Multiply_Vector2D_ByFloat(
     Float=&MergeHelper_FloatSwitch(0=1, 1=0, 2=?, 3=?, Index=&MakeIntFromFloat(Float=$User.VariantID), Num=::NewEnumerator0)
     Vector2D=$Particles.Initial.NeckGlowSize
    )
   ) {isSetParameters=true}
  ]
  rendererClasses [
   &NiagaraMeshRendererProperties
   &NiagaraSpriteRendererProperties
   &NiagaraLightRendererProperties
  ]
  renderers [
   {rendererClass=&NiagaraMeshRendererProperties, rendererIndex=0}
   {rendererClass=&NiagaraSpriteRendererProperties, rendererIndex=1}
   {rendererClass=&NiagaraSpriteRendererProperties, rendererIndex=2}
   {rendererClass=&NiagaraSpriteRendererProperties, rendererIndex=3}
   {rendererClass=&NiagaraLightRendererProperties, rendererIndex=4}
  ]
 }
}
summary {
 systemName=NS_CosmoHound_Idle_Glow
 emitters [
  {emitterName=CubicFlame, rendererClasses [&NiagaraSpriteRendererProperties, &NiagaraMeshRendererProperties]}
  {
   emitterName=NeckFlameMesh
   rendererClasses [
    &NiagaraMeshRendererProperties
    &NiagaraSpriteRendererProperties
    &NiagaraLightRendererProperties
   ]
  }
  {emitterName=Hunt_Lightning, rendererClasses [&NiagaraMeshRendererProperties]}
 ]
 userVariables [
  User.bIsFrontEnd {type=NiagaraBool}
  User.bIsFrontEndPreview {type=NiagaraBool}
  User.bisHeadHidden {type=NiagaraBool}
  User.VariantID {type=NiagaraFloat, defaultValue=0, description=""}
 ]
}
